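# Widget Blueprint based on FortWebLoginWidget that hosts a web login widget in the store frontend.
# DisplayWidget adds the supplied web widget to the transparent WebContent border, and DismissWidget clears that border.
# A centered circular throbber sits beneath WebContent in the same overlay, with no graph logic toggling it.
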WidgetBlueprint /FortniteCoreUI/Frontend/StoreMain/WebLogin
graphs [EventGraph {nodes [
   n1 = Event DisplayWidget!() -> (OutputDelegate, then, WebWidget:Widget Object Reference) {is=title K2Node_Event}
   > AddChild(self=Get WebContent() -> (WebContent:Border Object Reference), Content=@n1.WebWidget)
   Event DismissWidget() {is=title K2Node_Event}
   > ClearChildren(self=Get WebContent())
  ]}]
variables []
widgets {
 info {inheritedWidgetCount=0, parentClass=/Script/FortniteUI.FortWebLoginWidget, rootWidgetClass=&Overlay, widgetCount=9}
 widgets [
  Overlay_239 {widgetClassPath=&Overlay, widget=Overlay_239}
  SizeBox_148 {
   widgetClassPath=&SizeBox
   parent=Overlay_239
   widget=SizeBox_148
   properties {maxDesiredHeight=600, maxDesiredWidth=1024, minDesiredHeight=768, minDesiredWidth=1024, widthOverride=1280}
   slot {class=OverlaySlot, properties {horizontalAlignment=HAlign_Fill, verticalAlignment=VAlign_Fill}}
  }
  Overlay_929 {widgetClassPath=&Overlay, parent=SizeBox_148, widget=Overlay_929, slot {class=SizeBoxSlot}}
  VerticalBox_867 {
   widgetClassPath=&VerticalBox
   parent=Overlay_929
   widget=VerticalBox_867
   slot {class=OverlaySlot, properties {horizontalAlignment=HAlign_Fill, verticalAlignment=VAlign_Fill}}
  }
  Overlay_2891 {
   widgetClassPath=&Overlay
   parent=VerticalBox_867
   widget=Overlay_2891
   slot {class=VerticalBoxSlot, properties {size {sizeRule=Fill, value=1}}}
  }
  Border_12957 {
   widgetClassPath=&Border
   parent=Overlay_2891
   widget=Border_12957
   properties {brushColor=<rgba 0 0 0 0>, padding=<ltrb 0 15 0 10>}
   slot {class=OverlaySlot, properties {horizontalAlignment=HAlign_Fill, verticalAlignment=VAlign_Fill}}
  }
  SizeBox_0 {
   widgetClassPath=&SizeBox
   parent=Overlay_2891
   widget=SizeBox_0
   properties {bOverride_MinDesiredWidth=true, minDesiredWidth=1300}
   slot {class=OverlaySlot, properties {horizontalAlignment=HAlign_Center, verticalAlignment=VAlign_Center}}
  }
  Throbber {
   widgetClassPath=/Script/UMG.CircularThrobber
   bIsVariable=true
   parent=SizeBox_0
   widget=Throbber
   properties {numberOfPieces=20, period=1.5, radius=100}
   slot {class=SizeBoxSlot, properties {horizontalAlignment=HAlign_Center}}
  }
  WebContent {
   widgetClassPath=&Border
   bIsVariable=true
   parent=Overlay_2891
   widget=WebContent
   properties {brushColor=<rgba 1 1 1 0>, padding=<ltrb 0 0 0 0>}
   slot {class=OverlaySlot, properties {horizontalAlignment=HAlign_Fill, verticalAlignment=VAlign_Fill, padding=<ltrb 0 103 0 0>}}
  }
 ]
}
