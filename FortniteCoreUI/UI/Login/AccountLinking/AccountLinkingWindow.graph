WidgetBlueprint /FortniteCoreUI/UI/Login/AccountLinking/AccountLinkingWindow
graphs [EventGraph {links [], nodes []}]
variables []
widgets {
 info {inheritedWidgetCount=11, parentClass=/Script/FortniteUI.FortAccountLinkingWindow, rootWidgetClass=&Border, widgetCount=40}
 widgets [
  Border_0 {
   widgetClassPath=&Border
   widget=Border_0
   properties {
    horizontalAlignment=HAlign_Center
    background {
     bIsDynamicallyLoaded=false
     drawAs=NoDrawType
     imageType=NoImage
     mirroring=NoMirror
     resourceName=None
     resourceObject=null
     tiling=NoTile
     imageSize=<32 32>
     margin=<ltrb 0 0 0 0>
     outlineSettings {
      bUseBrushTransparency=false
      roundingType=HalfHeightRadius
      width=0
      color {colorUseRule=UseColor_Specified, specifiedColor=<rgba 0 0 0 0>}
      cornerRadii=<0 0 0 0>
     }
     tintColor {colorUseRule=UseColor_Specified, specifiedColor=<rgba 1 1 1 1>}
     uVRegion {bIsValid=false, max=<0 0>, min=<0 0>}
    }
    padding=<ltrb 0 0 0 0>
   }
  }
  SizeBoxMaxContainer {
   widgetClassPath=&SizeBox
   parent=Border_0
   widget=SizeBoxMaxContainer
   properties {bOverride_MinDesiredWidth=true, minDesiredWidth=1300}
   slot {class=BorderSlot, properties {horizontalAlignment=HAlign_Center, padding=<ltrb 0 0 0 0>}}
  }
  SizeBoxMinContainer {
   widgetClassPath=&SizeBox
   parent=SizeBoxMaxContainer
   widget=SizeBoxMinContainer
   properties {bOverride_MinDesiredWidth=true, minDesiredWidth=800}
   slot {class=SizeBoxSlot, properties {horizontalAlignment=HAlign_Center}}
  }
  Switcher_Main {
   widgetClassPath=&CommonWidgetSwitcherLegacy
   bInherited=true
   bIsVariable=true
   parent=SizeBoxMinContainer
   widget=Switcher_Main
   slot {class=SizeBoxSlot}
  }
  HaveEpicAccountBlur {
   widgetClassPath=&BackgroundBlur
   bInherited=true
   bIsVariable=true
   parent=Switcher_Main
   widget=HaveEpicAccountBlur
   properties {verticalAlignment=VAlign_Center}
   slot {class=WidgetSwitcherSlot}
  }
  CommonBorder_1 {
   widgetClassPath=&CommonBorder
   parent=HaveEpicAccountBlur
   widget=CommonBorder_1
   properties {bReducePaddingBySafezone=true, padding=<ltrb 0 0 0 0>}
   slot {class=BackgroundBlurSlot, properties {verticalAlignment=VAlign_Center, padding=<ltrb 0 0 0 0>}}
  }
  VerticalBoxHaveEpicAccount {
   widgetClassPath=&VerticalBox
   parent=CommonBorder_1
   widget=VerticalBoxHaveEpicAccount
   slot {class=BorderSlot, properties {padding=<ltrb 0 0 0 0>}}
  }
  CommonBorder_3 {
   widgetClassPath=&CommonBorder
   parent=VerticalBoxHaveEpicAccount
   widget=CommonBorder_3
   properties {padding=<ltrb 0 0 0 0>}
   slot {class=VerticalBoxSlot}
  }
  VerticalBox_1 {
   widgetClassPath=&VerticalBox
   parent=CommonBorder_3
   widget=VerticalBox_1
   slot {class=BorderSlot, properties {padding=<ltrb 0 0 0 0>}}
  }
  ImageLogo {
   widgetClassPath=&Image
   parent=VerticalBox_1
   widget=ImageLogo
   properties {brush {
     bIsDynamicallyLoaded=false
     drawAs=Image
     imageType=NoImage
     mirroring=NoMirror
     resourceName=None
     resourceObject=/FortniteCoreUI/UI/Login/Art/M_EpicLogo_DF
     tiling=NoTile
     imageSize=<120 120>
     margin=<ltrb 0 0 0 0>
     outlineSettings {
      bUseBrushTransparency=false
      roundingType=HalfHeightRadius
      width=0
      color {colorUseRule=UseColor_Specified, specifiedColor=<rgba 0 0 0 0>}
      cornerRadii=<0 0 0 0>
     }
     tintColor {colorUseRule=UseColor_Specified, specifiedColor=<rgba 1 1 1 1>}
     uVRegion {bIsValid=false, max=<0 0>, min=<0 0>}
    }}
   slot {class=VerticalBoxSlot, properties {horizontalAlignment=HAlign_Center, verticalAlignment=VAlign_Center}}
  }
  CommonTextTitle {
   widgetClassPath=&CommonTextBlock
   parent=VerticalBox_1
   widget=CommonTextTitle
   properties {
    style=/FortUILibrary/Styles/Typography/Foundation/TextStyle-Header-L.TextStyle-Header-L_C
    text=PLAY WITH AN EPIC ACCOUNT
    wrapTextAt=850
    font {
     bForceMonospaced=false
     bMaterialIsStencil=false
     fontMaterial=null
     fontObject=&Burbank
     letterSpacing=0
     monospacedWidth=1
     size=48
     skewAmount=0
     typefaceFontName=Burbank-Black
     outlineSettings {
      bApplyOutlineToDropShadows=false
      bMiteredCorners=false
      bSeparateFillAlpha=false
      outlineMaterial=null
      outlineSize=0
      outlineColor=<rgba 0 0 0 1>
     }
    }
    shadowOffset=<0 0>
   }
   slot {class=VerticalBoxSlot, properties {horizontalAlignment=HAlign_Center, verticalAlignment=VAlign_Center}}
  }
  CommonTextYouGet {
   widgetClassPath=&CommonTextBlock
   parent=VerticalBoxHaveEpicAccount
   widget=CommonTextYouGet
   properties {
    style=&TextStyle-Header-L_C
    text=YOU GET
    visibility=Collapsed
    font {
     bForceMonospaced=false
     bMaterialIsStencil=false
     fontMaterial=null
     fontObject=&Burbank
     letterSpacing=0
     monospacedWidth=1
     size=48
     skewAmount=0
     typefaceFontName=Burbank-Black
     outlineSettings {
      bApplyOutlineToDropShadows=false
      bMiteredCorners=false
      bSeparateFillAlpha=false
      outlineMaterial=null
      outlineSize=0
      outlineColor=<rgba 0 0 0 1>
     }
    }
    shadowOffset=<0 0>
   }
   slot {class=VerticalBoxSlot, properties {padding=<ltrb 20 30 0 0>}}
  }
  Image_0 {
   widgetClassPath=&Image
   parent=VerticalBoxHaveEpicAccount
   widget=Image_0
   properties {
    visibility=Collapsed
    brush {
     bIsDynamicallyLoaded=false
     drawAs=Image
     imageType=NoImage
     mirroring=NoMirror
     resourceName=None
     resourceObject=/Game/UI/Foundation/Textures/Rule/Rule-Dotted-8
     tiling=Horizontal
     imageSize=<6 6>
     margin=<ltrb 0 0 0 0>
     outlineSettings {
      bUseBrushTransparency=false
      roundingType=HalfHeightRadius
      width=0
      color {colorUseRule=UseColor_Specified, specifiedColor=<rgba 0 0 0 0>}
      cornerRadii=<0 0 0 0>
     }
     tintColor {colorUseRule=UseColor_Specified, specifiedColor=<rgba 1 1 1 1>}
     uVRegion {bIsValid=false, max=<0 0>, min=<0 0>}
    }
   }
   slot {class=VerticalBoxSlot, properties {padding=<ltrb 20 0 20 0>}}
  }
  HorizontalBoxItem1 {
   widgetClassPath=&HorizontalBox
   parent=VerticalBoxHaveEpicAccount
   widget=HorizontalBoxItem1
   properties {visibility=Collapsed}
   slot {class=VerticalBoxSlot, properties {padding=<ltrb 20 30 0 0>}}
  }
  ImageBullet1 {
   widgetClassPath=&Image
   parent=HorizontalBoxItem1
   widget=ImageBullet1
   properties {brush {
     bIsDynamicallyLoaded=false
     drawAs=Image
     imageType=NoImage
     mirroring=NoMirror
     resourceName=None
     resourceObject=null
     tiling=NoTile
     imageSize=<96 96>
     margin=<ltrb 0 0 0 0>
     outlineSettings {
      bUseBrushTransparency=false
      roundingType=HalfHeightRadius
      width=0
      color {colorUseRule=UseColor_Specified, specifiedColor=<rgba 0 0 0 0>}
      cornerRadii=<0 0 0 0>
     }
     tintColor {colorUseRule=UseColor_Specified, specifiedColor=<rgba 1 1 1 1>}
     uVRegion {bIsValid=false, max=<0 0>, min=<0 0>}
    }}
   slot {class=HorizontalBoxSlot, properties {horizontalAlignment=HAlign_Left, verticalAlignment=VAlign_Center, padding=<ltrb 0 0 40 0>}}
  }
  CommonTextItemText1 {
   widgetClassPath=&CommonTextBlock
   parent=HorizontalBoxItem1
   widget=CommonTextItemText1
   properties {
    style=/FortUILibrary/Styles/Typography/Foundation/TextStyle-Base-S-Black.TextStyle-Base-S-Black_C
    text="Lorem ipsum dolor sit amet, consectetur adipiscing elit. Aenean euismod bibendum laoreet."
    wrapTextAt=600
    colorAndOpacity {colorUseRule=UseColor_Specified, specifiedColor=<rgba 0 0 0 1>}
    font {
     bForceMonospaced=false
     bMaterialIsStencil=false
     fontMaterial=null
     fontObject=&NotoSans
     letterSpacing=0
     monospacedWidth=1
     size=18
     skewAmount=0
     typefaceFontName=NotoSans-Regular
     outlineSettings {
      bApplyOutlineToDropShadows=false
      bMiteredCorners=false
      bSeparateFillAlpha=false
      outlineMaterial=null
      outlineSize=0
      outlineColor=<rgba 0 0 0 1>
     }
    }
    shadowOffset=<0 0>
   }
   slot {class=HorizontalBoxSlot, properties {verticalAlignment=VAlign_Center}}
  }
  HorizontalBoxItem2 {
   widgetClassPath=&HorizontalBox
   parent=VerticalBoxHaveEpicAccount
   widget=HorizontalBoxItem2
   properties {visibility=Collapsed}
   slot {class=VerticalBoxSlot, properties {padding=<ltrb 20 30 0 0>}}
  }
  ImageBullet2 {
   widgetClassPath=&Image
   parent=HorizontalBoxItem2
   widget=ImageBullet2
   properties {brush {
     bIsDynamicallyLoaded=false
     drawAs=Image
     imageType=NoImage
     mirroring=NoMirror
     resourceName=None
     resourceObject=null
     tiling=NoTile
     imageSize=<96 96>
     margin=<ltrb 0 0 0 0>
     outlineSettings {
      bUseBrushTransparency=false
      roundingType=HalfHeightRadius
      width=0
      color {colorUseRule=UseColor_Specified, specifiedColor=<rgba 0 0 0 0>}
      cornerRadii=<0 0 0 0>
     }
     tintColor {colorUseRule=UseColor_Specified, specifiedColor=<rgba 1 1 1 1>}
     uVRegion {bIsValid=false, max=<0 0>, min=<0 0>}
    }}
   slot {class=HorizontalBoxSlot, properties {horizontalAlignment=HAlign_Left, verticalAlignment=VAlign_Center, padding=<ltrb 0 0 40 0>}}
  }
  CommonTextItemText2 {
   widgetClassPath=&CommonTextBlock
   parent=HorizontalBoxItem2
   widget=CommonTextItemText2
   properties {
    style=&TextStyle-Base-S-Black_C
    text="Lorem ipsum dolor sit amet, consectetur adipiscing elit. Aenean euismod bibendum laoreet."
    wrapTextAt=600
    colorAndOpacity {colorUseRule=UseColor_Specified, specifiedColor=<rgba 0 0 0 1>}
    font {
     bForceMonospaced=false
     bMaterialIsStencil=false
     fontMaterial=null
     fontObject=&NotoSans
     letterSpacing=0
     monospacedWidth=1
     size=18
     skewAmount=0
     typefaceFontName=NotoSans-Regular
     outlineSettings {
      bApplyOutlineToDropShadows=false
      bMiteredCorners=false
      bSeparateFillAlpha=false
      outlineMaterial=null
      outlineSize=0
      outlineColor=<rgba 0 0 0 1>
     }
    }
    shadowOffset=<0 0>
   }
   slot {class=HorizontalBoxSlot, properties {verticalAlignment=VAlign_Center}}
  }
  HorizontalBoxItem3 {
   widgetClassPath=&HorizontalBox
   parent=VerticalBoxHaveEpicAccount
   widget=HorizontalBoxItem3
   properties {visibility=Collapsed}
   slot {class=VerticalBoxSlot, properties {padding=<ltrb 20 30 0 30>}}
  }
  ImageBullet3 {
   widgetClassPath=&Image
   parent=HorizontalBoxItem3
   widget=ImageBullet3
   properties {brush {
     bIsDynamicallyLoaded=false
     drawAs=Image
     imageType=NoImage
     mirroring=NoMirror
     resourceName=None
     resourceObject=null
     tiling=NoTile
     imageSize=<96 96>
     margin=<ltrb 0 0 0 0>
     outlineSettings {
      bUseBrushTransparency=false
      roundingType=HalfHeightRadius
      width=0
      color {colorUseRule=UseColor_Specified, specifiedColor=<rgba 0 0 0 0>}
      cornerRadii=<0 0 0 0>
     }
     tintColor {colorUseRule=UseColor_Specified, specifiedColor=<rgba 1 1 1 1>}
     uVRegion {bIsValid=false, max=<0 0>, min=<0 0>}
    }}
   slot {class=HorizontalBoxSlot, properties {horizontalAlignment=HAlign_Left, verticalAlignment=VAlign_Center, padding=<ltrb 0 0 40 0>}}
  }
  CommonTextItemText3 {
   widgetClassPath=&CommonTextBlock
   parent=HorizontalBoxItem3
   widget=CommonTextItemText3
   properties {
    style=&TextStyle-Base-S-Black_C
    text="Lorem ipsum dolor sit amet, consectetur adipiscing elit. Aenean euismod bibendum laoreet."
    wrapTextAt=600
    colorAndOpacity {colorUseRule=UseColor_Specified, specifiedColor=<rgba 0 0 0 1>}
    font {
     bForceMonospaced=false
     bMaterialIsStencil=false
     fontMaterial=null
     fontObject=&NotoSans
     letterSpacing=0
     monospacedWidth=1
     size=18
     skewAmount=0
     typefaceFontName=NotoSans-Regular
     outlineSettings {
      bApplyOutlineToDropShadows=false
      bMiteredCorners=false
      bSeparateFillAlpha=false
      outlineMaterial=null
      outlineSize=0
      outlineColor=<rgba 0 0 0 1>
     }
    }
    shadowOffset=<0 0>
   }
   slot {class=HorizontalBoxSlot, properties {verticalAlignment=VAlign_Center}}
  }
  VerticalBox_0 {
   widgetClassPath=&VerticalBox
   parent=VerticalBoxHaveEpicAccount
   widget=VerticalBox_0
   slot {class=VerticalBoxSlot, properties {verticalAlignment=VAlign_Bottom, padding=<ltrb 0 12 0 12>}}
  }
  Button_Signup {
   widgetClassPath=&IconTextButton
   bInherited=true
   bIsVariable=true
   parent=VerticalBox_0
   widget=Button_Signup
   properties {
    buttonText=SIGN UP
    contentAlignment=HAlign_Center
    minHeight=100
    minWidth=275
    onButtonBaseClicked=(&self:WidgetTree.Button_Signup.HandleOnButtonClicked)
    onButtonBaseDoubleClicked=(&self:WidgetTree.Button_Signup.HandleOnButtonDoubleClicked)
    onButtonBaseHovered=(&self:WidgetTree.Button_Signup.HandleOnButtonHovered)
    onButtonBaseUnhovered=(&self:WidgetTree.Button_Signup.HandleOnButtonUnhovered)
    onSelectedChangedBase=(&self:WidgetTree.Button_Signup.HandleOnSelectedChanged)
    style=/FortUILibrary/Foundation/Buttons/ButtonStyles/Athena/Athena_ButtonStyle_AngledBlueMenuButton.Athena_ButtonStyle_AngledBlueMenuButton_C
   }
   slot {class=VerticalBoxSlot, properties {padding=<ltrb 20 20 20 20>}}
  }
  Button_Login {
   widgetClassPath=&IconTextButton
   bInherited=true
   bIsVariable=true
   parent=VerticalBox_0
   widget=Button_Login
   properties {
    buttonText=LOG IN
    contentAlignment=HAlign_Center
    minHeight=100
    minWidth=275
    onButtonBaseClicked=(&self:WidgetTree.Button_Login.HandleOnButtonClicked)
    onButtonBaseDoubleClicked=(&self:WidgetTree.Button_Login.HandleOnButtonDoubleClicked)
    onButtonBaseHovered=(&self:WidgetTree.Button_Login.HandleOnButtonHovered)
    onButtonBaseUnhovered=(&self:WidgetTree.Button_Login.HandleOnButtonUnhovered)
    onSelectedChangedBase=(&self:WidgetTree.Button_Login.HandleOnSelectedChanged)
    style=&Athena_ButtonStyle_AngledBlueMenuButton_C
   }
   slot {class=VerticalBoxSlot, properties {padding=<ltrb 20 20 20 20>}}
  }
  Image_1 {
   widgetClassPath=&Image
   parent=VerticalBox_0
   widget=Image_1
   properties {brush {
     bIsDynamicallyLoaded=false
     drawAs=Image
     imageType=NoImage
     mirroring=NoMirror
     resourceName=None
     resourceObject=&Rule-Dotted-8
     tiling=Horizontal
     imageSize=<6 6>
     margin=<ltrb 0 0 0 0>
     outlineSettings {
      bUseBrushTransparency=false
      roundingType=HalfHeightRadius
      width=0
      color {colorUseRule=UseColor_Specified, specifiedColor=<rgba 0 0 0 0>}
      cornerRadii=<0 0 0 0>
     }
     tintColor {colorUseRule=UseColor_Specified, specifiedColor=<rgba 1 1 1 1>}
     uVRegion {bIsValid=false, max=<0 0>, min=<0 0>}
    }}
   slot {class=VerticalBoxSlot, properties {padding=<ltrb 0 50 0 50>}}
  }
  Button_NoThanks {
   widgetClassPath=&IconTextButton
   bInherited=true
   bIsVariable=true
   parent=VerticalBox_0
   widget=Button_NoThanks
   properties {
    buttonText=NO THANKS
    contentAlignment=HAlign_Center
    minHeight=100
    minWidth=275
    onButtonBaseClicked=(&self:WidgetTree.Button_NoThanks.HandleOnButtonClicked)
    onButtonBaseDoubleClicked=(&self:WidgetTree.Button_NoThanks.HandleOnButtonDoubleClicked)
    onButtonBaseHovered=(&self:WidgetTree.Button_NoThanks.HandleOnButtonHovered)
    onButtonBaseUnhovered=(&self:WidgetTree.Button_NoThanks.HandleOnButtonUnhovered)
    onSelectedChangedBase=(&self:WidgetTree.Button_NoThanks.HandleOnSelectedChanged)
    style=/FortUILibrary/Foundation/Buttons/ButtonStyles/TopBar/ButtonStyle-Outline-M.ButtonStyle-Outline-M_C
   }
   slot {class=VerticalBoxSlot, properties {padding=<ltrb 20 20 20 20>}}
  }
  SkipSignInBlur {
   widgetClassPath=&BackgroundBlur
   bInherited=true
   bIsVariable=true
   parent=Switcher_Main
   widget=SkipSignInBlur
   properties {verticalAlignment=VAlign_Center}
   slot {class=WidgetSwitcherSlot}
  }
  CommonBorder_2 {
   widgetClassPath=&CommonBorder
   parent=SkipSignInBlur
   widget=CommonBorder_2
   properties {bReducePaddingBySafezone=true, padding=<ltrb 0 0 0 0>}
   slot {class=BackgroundBlurSlot, properties {verticalAlignment=VAlign_Center, padding=<ltrb 0 0 0 0>}}
  }
  VerticalBoxSkipSignIn {
   widgetClassPath=&VerticalBox
   parent=CommonBorder_2
   widget=VerticalBoxSkipSignIn
   slot {class=BorderSlot, properties {padding=<ltrb 0 0 0 0>}}
  }
  CommonBorder_4 {
   widgetClassPath=&CommonBorder
   parent=VerticalBoxSkipSignIn
   widget=CommonBorder_4
   properties {horizontalAlignment=HAlign_Center, padding=<ltrb 0 0 0 0>}
   slot {class=VerticalBoxSlot}
  }
  TextTitleText {
   widgetClassPath=&CommonTextBlock
   parent=CommonBorder_4
   widget=TextTitleText
   properties {
    style=&TextStyle-Header-L_C
    text=ARE YOU SURE?
    font {
     bForceMonospaced=false
     bMaterialIsStencil=false
     fontMaterial=null
     fontObject=&Burbank
     letterSpacing=0
     monospacedWidth=1
     size=48
     skewAmount=0
     typefaceFontName=Burbank-Black
     outlineSettings {
      bApplyOutlineToDropShadows=false
      bMiteredCorners=false
      bSeparateFillAlpha=false
      outlineMaterial=null
      outlineSize=0
      outlineColor=<rgba 0 0 0 1>
     }
    }
    shadowOffset=<0 0>
   }
   slot {class=BorderSlot, properties {horizontalAlignment=HAlign_Center, padding=<ltrb 0 0 0 0>}}
  }
  CommonTextSkipSigninText {
   widgetClassPath=&CommonTextBlock
   parent=VerticalBoxSkipSignIn
   widget=CommonTextSkipSigninText
   properties {
    style=/FortUILibrary/Styles/Typography/Foundation/TextStyle-Base-M-B.TextStyle-Base-M-B_C
    text=You will not be able to link your account in the future if you proceed without signing up or logging in.
    wrapTextAt=800
    font {
     bForceMonospaced=false
     bMaterialIsStencil=false
     fontMaterial=null
     fontObject=&NotoSans
     letterSpacing=0
     monospacedWidth=1
     size=24
     skewAmount=0
     typefaceFontName=NotoSans-Bold
     outlineSettings {
      bApplyOutlineToDropShadows=false
      bMiteredCorners=false
      bSeparateFillAlpha=false
      outlineMaterial=null
      outlineSize=0
      outlineColor=<rgba 0 0 0 1>
     }
    }
    shadowOffset=<0 0>
   }
   slot {class=VerticalBoxSlot, properties {horizontalAlignment=HAlign_Center, padding=<ltrb 20 20 20 20>}}
  }
  VerticalBox_2 {widgetClassPath=&VerticalBox, parent=VerticalBoxSkipSignIn, widget=VerticalBox_2, slot {class=VerticalBoxSlot}}
  Button_SkipSignInSignUp {
   widgetClassPath=&IconTextButton
   bInherited=true
   bIsVariable=true
   parent=VerticalBox_2
   widget=Button_SkipSignInSignUp
   properties {
    buttonText=SIGN UP
    contentAlignment=HAlign_Center
    minHeight=100
    minWidth=275
    onButtonBaseClicked=(&self:WidgetTree.Button_SkipSignInSignUp.HandleOnButtonClicked)
    onButtonBaseDoubleClicked=(&self:WidgetTree.Button_SkipSignInSignUp.HandleOnButtonDoubleClicked)
    onButtonBaseHovered=(&self:WidgetTree.Button_SkipSignInSignUp.HandleOnButtonHovered)
    onButtonBaseUnhovered=(&self:WidgetTree.Button_SkipSignInSignUp.HandleOnButtonUnhovered)
    onSelectedChangedBase=(&self:WidgetTree.Button_SkipSignInSignUp.HandleOnSelectedChanged)
    style=&ButtonStyle-Outline-M_C
   }
   slot {class=VerticalBoxSlot, properties {padding=<ltrb 0 20 0 20>}}
  }
  Button_SkipSignInLogin {
   widgetClassPath=&IconTextButton
   bInherited=true
   bIsVariable=true
   parent=VerticalBox_2
   widget=Button_SkipSignInLogin
   properties {
    buttonText=LOG IN
    contentAlignment=HAlign_Center
    minHeight=100
    minWidth=275
    onButtonBaseClicked=(&self:WidgetTree.Button_SkipSignInLogin.HandleOnButtonClicked)
    onButtonBaseDoubleClicked=(&self:WidgetTree.Button_SkipSignInLogin.HandleOnButtonDoubleClicked)
    onButtonBaseHovered=(&self:WidgetTree.Button_SkipSignInLogin.HandleOnButtonHovered)
    onButtonBaseUnhovered=(&self:WidgetTree.Button_SkipSignInLogin.HandleOnButtonUnhovered)
    onSelectedChangedBase=(&self:WidgetTree.Button_SkipSignInLogin.HandleOnSelectedChanged)
    style=&ButtonStyle-Outline-M_C
   }
   slot {class=VerticalBoxSlot, properties {padding=<ltrb 0 20 0 20>}}
  }
  Spacer {
   widgetClassPath=&Image
   parent=VerticalBoxSkipSignIn
   widget=Spacer
   properties {brush {
     bIsDynamicallyLoaded=false
     drawAs=Image
     imageType=NoImage
     mirroring=NoMirror
     resourceName=None
     resourceObject=&Rule-Dotted-8
     tiling=Horizontal
     imageSize=<6 6>
     margin=<ltrb 0 0 0 0>
     outlineSettings {
      bUseBrushTransparency=false
      roundingType=HalfHeightRadius
      width=0
      color {colorUseRule=UseColor_Specified, specifiedColor=<rgba 0 0 0 0>}
      cornerRadii=<0 0 0 0>
     }
     tintColor {colorUseRule=UseColor_Specified, specifiedColor=<rgba 1 1 1 1>}
     uVRegion {bIsValid=false, max=<0 0>, min=<0 0>}
    }}
   slot {class=VerticalBoxSlot, properties {padding=<ltrb 0 50 0 50>}}
  }
  Switcher_NoThanks {
   widgetClassPath=&CommonWidgetSwitcherLegacy
   bInherited=true
   bIsVariable=true
   parent=VerticalBoxSkipSignIn
   widget=Switcher_NoThanks
   slot {class=VerticalBoxSlot, properties {horizontalAlignment=HAlign_Center, padding=<ltrb 0 0 0 30>}}
  }
  Throbber_LoginDelay {
   widgetClassPath=/Script/UMG.CircularThrobber
   bInherited=true
   bIsVariable=true
   parent=Switcher_NoThanks
   widget=Throbber_LoginDelay
   properties {radius=32}
   slot {class=WidgetSwitcherSlot}
  }
  Button_SkipSignInNoThanks {
   widgetClassPath=&IconTextButton
   bInherited=true
   bIsVariable=true
   parent=Switcher_NoThanks
   widget=Button_SkipSignInNoThanks
   properties {
    buttonText="NO THANKS, PLAY FORTNITE"
    contentAlignment=HAlign_Center
    minHeight=100
    minWidth=275
    onButtonBaseClicked=(&self:WidgetTree.Button_SkipSignInNoThanks.HandleOnButtonClicked)
    onButtonBaseDoubleClicked=(&self:WidgetTree.Button_SkipSignInNoThanks.HandleOnButtonDoubleClicked)
    onButtonBaseHovered=(&self:WidgetTree.Button_SkipSignInNoThanks.HandleOnButtonHovered)
    onButtonBaseUnhovered=(&self:WidgetTree.Button_SkipSignInNoThanks.HandleOnButtonUnhovered)
    onSelectedChangedBase=(&self:WidgetTree.Button_SkipSignInNoThanks.HandleOnSelectedChanged)
    style=&ButtonStyle-Outline-M_C
   }
   slot {class=WidgetSwitcherSlot}
  }
 ]
}
