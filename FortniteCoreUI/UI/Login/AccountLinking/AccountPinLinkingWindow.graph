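# Login screen widget, based on FortAccountPinLinkingWindow, that prompts the player to link an Epic Games account or skip.
# Switcher_Main holds three pages: link/skip prompt, PIN grant code info, and skip confirmation with a nested throbber/No Thanks switcher.
# After each page change, the graph waits half a second, then focuses Link on page 0 or Go Back on page 2.
# Link, Skip and Go Back clicks only fire element-interacted events in this graph.
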WidgetBlueprint /FortniteCoreUI/UI/Login/AccountLinking/AccountPinLinkingWindow
graphs [EventGraph {nodes [
   n1 = "On Active Widget Changed (Switcher_Main)"!() -> (OutputDelegate, then, ActiveWidget:Widget Object Reference, ActiveWidgetIndex:int=0) {is=title K2Node_ComponentBoundEvent}
   > Delay(Duration=0.5)
   > n2 = Switch on Int!(execute:Exec, Selection:int=@n1.ActiveWidgetIndex) -> (Default:Exec, 0:Exec, 1:Exec, 2:Exec)
   SetFocus(execute=@n2.0, self=Get Button_GoLinking() -> (Button_GoLinking:Common Button Legacy Object Reference))
   SetFocus(execute=@n2.2, self=Get Button_SkipGoBackLinking() -> (Button_SkipGoBackLinking:Common Button Legacy Object Reference))
   n3 = "On Button Clicked (Button_GoLinking)"!() -> (OutputDelegate, then, Button:Common Button Legacy Object Reference) {is=title K2Node_ComponentBoundEvent}
   > Fire Event Element Interacted(Element=@n3.Button)
   n4 = "On Button Clicked (Button_SkipLinking)"!() -> (OutputDelegate, then, Button:Common Button Legacy Object Reference) {is=title K2Node_ComponentBoundEvent}
   > Fire Event Element Interacted(Element=@n4.Button)
   n5 = "On Button Clicked (Button_SkipGoBackLinking)"!() -> (OutputDelegate, then, Button:Common Button Legacy Object Reference) {is=title K2Node_ComponentBoundEvent}
   > Fire Event Element Interacted(Element=@n5.Button)
   Event PreConstruct()
   Event Construct()
   Event Tick()
  ]}]
variables []
widgets {
 info {inheritedWidgetCount=12, parentClass=/Script/FortniteUI.FortAccountPinLinkingWindow, rootWidgetClass=&SafeZone, widgetCount=27}
 widgets [
  SafeZone {widgetClassPath=&SafeZone, bIsVariable=true, widget=SafeZone}
  Overlay_0 {widgetClassPath=&Overlay, parent=SafeZone, widget=Overlay_0, slot {class=SafeZoneSlot}}
  BorderBackBtn {
   widgetClassPath=&Border
   parent=Overlay_0
   widget=BorderBackBtn
   properties {
    horizontalAlignment=HAlign_Left
    verticalAlignment=VAlign_Top
    background {
     bIsDynamicallyLoaded=false
     drawAs=NoDrawType
     imageType=NoImage
     mirroring=NoMirror
     resourceName=None
     resourceObject=null
     tiling=NoTile
     imageSize=<32 32>
     margin=<ltrb 0 0 0 0>
     outlineSettings {
      bUseBrushTransparency=false
      roundingType=HalfHeightRadius
      width=0
      color {colorUseRule=UseColor_Specified, specifiedColor=<rgba 0 0 0 0>}
      cornerRadii=<0 0 0 0>
     }
     tintColor {colorUseRule=UseColor_Specified, specifiedColor=<rgba 1 1 1 1>}
     uVRegion {bIsValid=false, max=<0 0>, min=<0 0>}
    }
    padding=<ltrb 0 0 0 0>
   }
   slot {class=OverlaySlot}
  }
  Button_CancelPinRequest {
   widgetClassPath=&IconTextButton
   bInherited=true
   bIsVariable=true
   parent=BorderBackBtn
   widget=Button_CancelPinRequest
   properties {
    bHideInputActionWithKeyboard=true
    bIsFocusable=false
    buttonText=""
    contentAlignment=HAlign_Right
    onButtonBaseClicked=(&self:WidgetTree.Button_CancelPinRequest.HandleOnButtonClicked)
    onButtonBaseDoubleClicked=(&self:WidgetTree.Button_CancelPinRequest.HandleOnButtonDoubleClicked)
    onButtonBaseHovered=(&self:WidgetTree.Button_CancelPinRequest.HandleOnButtonHovered)
    onButtonBaseUnhovered=(&self:WidgetTree.Button_CancelPinRequest.HandleOnButtonUnhovered)
    onSelectedChangedBase=(&self:WidgetTree.Button_CancelPinRequest.HandleOnSelectedChanged)
    style=/FortUILibrary/Foundation/Buttons/ButtonStyles/ButtonStyle-TextOnlyBase_Empty.ButtonStyle-TextOnlyBase_Empty_C
    visibility=Hidden
    iconBrush {
     bIsDynamicallyLoaded=false
     drawAs=Image
     imageType=NoImage
     mirroring=NoMirror
     resourceName=None
     resourceObject=/FortniteCoreUI/UI/Login/Art/M_CornerBackArrow
     tiling=NoTile
     imageSize=<112 112>
     margin=<ltrb 0 0 0 0>
     outlineSettings {
      bUseBrushTransparency=false
      roundingType=HalfHeightRadius
      width=0
      color {colorUseRule=UseColor_Specified, specifiedColor=<rgba 0 0 0 0>}
      cornerRadii=<0 0 0 1>
     }
     tintColor {colorUseRule=UseColor_Specified, specifiedColor=<rgba 1 1 1 1>}
     uVRegion {bIsValid=false, max=<0 0>, min=<0 0>}
    }
    triggeringInputAction {dataTable=&InputActionDataTable, rowName=GenericBack}
   }
   slot {class=BorderSlot, properties {horizontalAlignment=HAlign_Left, verticalAlignment=VAlign_Top, padding=<ltrb 0 0 0 0>}}
  }
  VB_MainContent {
   widgetClassPath=&VerticalBox
   parent=Overlay_0
   widget=VB_MainContent
   slot {class=OverlaySlot, properties {horizontalAlignment=HAlign_Fill, verticalAlignment=VAlign_Fill, padding=<ltrb 0 64 0 64>}}
  }
  Logo {
   widgetClassPath=&Image
   parent=VB_MainContent
   widget=Logo
   properties {brush {
     bIsDynamicallyLoaded=false
     drawAs=Image
     imageType=FullColor
     mirroring=NoMirror
     resourceName=None
     resourceObject=/FortniteCoreUI/UI/Login/Art/MI_UI_Logo_Fortnite
     tiling=NoTile
     imageSize=<460 128>
     margin=<ltrb 0 0 0 0>
     outlineSettings {
      bUseBrushTransparency=false
      roundingType=HalfHeightRadius
      width=0
      color {colorUseRule=UseColor_Specified, specifiedColor=<rgba 0 0 0 0>}
      cornerRadii=<0 0 0 0>
     }
     tintColor {colorUseRule=UseColor_Specified, specifiedColor=<rgba 1 1 1 1>}
     uVRegion {bIsValid=false, max=<0 0>, min=<0 0>}
    }}
   slot {class=VerticalBoxSlot, properties {horizontalAlignment=HAlign_Center, verticalAlignment=VAlign_Top, padding=<ltrb 0 0 0 64>}}
  }
  Switcher_Main {
   widgetClassPath=&CommonWidgetSwitcherLegacy
   bInherited=true
   bIsVariable=true
   parent=VB_MainContent
   widget=Switcher_Main
   slot {class=VerticalBoxSlot, properties {horizontalAlignment=HAlign_Center, size {sizeRule=Fill, value=1}}}
  }
  Blur_InitialRequest {
   widgetClassPath=&BackgroundBlur
   bInherited=true
   bIsVariable=true
   parent=Switcher_Main
   widget=Blur_InitialRequest
   properties {verticalAlignment=VAlign_Center}
   slot {class=WidgetSwitcherSlot, properties {horizontalAlignment=HAlign_Center}}
  }
  VB_SkipSignIn {
   widgetClassPath=&VerticalBox
   parent=Blur_InitialRequest
   widget=VB_SkipSignIn
   slot {class=BackgroundBlurSlot, properties {verticalAlignment=VAlign_Center, padding=<ltrb 0 0 0 0>}}
  }
  HorizontalBox_0 {
   widgetClassPath=&HorizontalBox
   parent=VB_SkipSignIn
   widget=HorizontalBox_0
   slot {class=VerticalBoxSlot, properties {horizontalAlignment=HAlign_Center}}
  }
  SizeBox_0 {
   widgetClassPath=&SizeBox
   parent=HorizontalBox_0
   widget=SizeBox_0
   properties {bOverride_WidthOverride=true, widthOverride=550}
   slot {class=HorizontalBoxSlot, properties {verticalAlignment=VAlign_Top}}
  }
  VerticalBox_1 {widgetClassPath=&VerticalBox, parent=SizeBox_0, widget=VerticalBox_1, slot {class=SizeBoxSlot}}
  Button_GoLinking {
   widgetClassPath=&IconTextButton
   bInherited=true
   bIsVariable=true
   parent=VerticalBox_1
   widget=Button_GoLinking
   properties {
    buttonText=Link An Account
    onButtonBaseClicked=(&self:WidgetTree.Button_GoLinking.HandleOnButtonClicked)
    onButtonBaseDoubleClicked=(&self:WidgetTree.Button_GoLinking.HandleOnButtonDoubleClicked)
    onButtonBaseHovered=(&self:WidgetTree.Button_GoLinking.HandleOnButtonHovered)
    onButtonBaseUnhovered=(&self:WidgetTree.Button_GoLinking.HandleOnButtonUnhovered)
    onSelectedChangedBase=(&self:WidgetTree.Button_GoLinking.HandleOnSelectedChanged)
    style=/FortUILibrary/Foundation/Buttons/ButtonStyles/ButtonStyle_Feature_L_Yellow.ButtonStyle_Feature_L_Yellow_C
    triggeringInputAction {dataTable=&InputActionDataTable, rowName=Confirm}
   }
   slot {class=VerticalBoxSlot, properties {verticalAlignment=VAlign_Top}}
  }
  CommonTextBlock_117 {
   widgetClassPath=&CommonTextBlock
   parent=VerticalBox_1
   widget=CommonTextBlock_117
   properties {
    autoWrapText=true
    justification=Center
    text=Been Here Before?
    textTransformPolicy=ToUpper
    colorAndOpacity {colorUseRule=UseColor_Specified, specifiedColor=<rgba 0.991 0.991 0.991 1>}
    font {
     bForceMonospaced=false
     bMaterialIsStencil=false
     fontMaterial=null
     fontObject=&BurbankSmall
     letterSpacing=0
     monospacedWidth=1
     size=28
     skewAmount=0
     typefaceFontName=Black
     outlineSettings {
      bApplyOutlineToDropShadows=false
      bMiteredCorners=false
      bSeparateFillAlpha=false
      outlineMaterial=null
      outlineSize=0
      outlineColor=<rgba 0 0 0 1>
     }
    }
    shadowOffset=<1 2>
   }
   slot {class=VerticalBoxSlot, properties {verticalAlignment=VAlign_Top, padding=<ltrb 0 40 0 8>}}
  }
  CommonTextBlock {
   widgetClassPath=&CommonTextBlock
   parent=VerticalBox_1
   widget=CommonTextBlock
   properties {
    autoWrapText=true
    justification=Center
    text=Link an Epic Games account to access your progression and game data wherever you play.
    font {
     bForceMonospaced=false
     bMaterialIsStencil=false
     fontMaterial=null
     fontObject=&BurbankSmall
     letterSpacing=0
     monospacedWidth=1
     size=24
     skewAmount=0
     typefaceFontName=Bold
     outlineSettings {
      bApplyOutlineToDropShadows=false
      bMiteredCorners=false
      bSeparateFillAlpha=false
      outlineMaterial=null
      outlineSize=0
      outlineColor=<rgba 0 0 0 1>
     }
    }
    shadowOffset=<1 2>
   }
   slot {class=VerticalBoxSlot, properties {verticalAlignment=VAlign_Top}}
  }
  Button_SkipLinking {
   widgetClassPath=/FortUILibrary/Foundation/Buttons/SoloButton
   bInherited=true
   bIsVariable=true
   parent=VB_SkipSignIn
   widget=Button_SkipLinking
   properties {
    button Description=SKIP THIS
    onButtonBaseClicked=(&self:WidgetTree.Button_SkipLinking.HandleOnButtonClicked)
    onButtonBaseDoubleClicked=(&self:WidgetTree.Button_SkipLinking.HandleOnButtonDoubleClicked)
    onButtonBaseHovered=(&self:WidgetTree.Button_SkipLinking.HandleOnButtonHovered)
    onButtonBaseUnhovered=(&self:WidgetTree.Button_SkipLinking.HandleOnButtonUnhovered)
    onSelectedChangedBase=(&self:WidgetTree.Button_SkipLinking.HandleOnSelectedChanged)
    triggeringInputAction {dataTable=&InputActionDataTable, rowName=GenericFaceButtonLeft}
   }
   slot {class=VerticalBoxSlot, properties {horizontalAlignment=HAlign_Center, padding=<ltrb 0 64 0 0>}}
  }
  Blur_PinGrantCodeInfo {
   widgetClassPath=&BackgroundBlur
   bInherited=true
   bIsVariable=true
   parent=Switcher_Main
   widget=Blur_PinGrantCodeInfo
   slot {class=WidgetSwitcherSlot, properties {horizontalAlignment=HAlign_Center}}
  }
  PinGrantCodeInfoWidget {
   widgetClassPath=&/PinGrantCodeInfoWidget
   bInherited=true
   bIsVariable=true
   parent=Blur_PinGrantCodeInfo
   widget=PinGrantCodeInfoWidget
   slot {class=BackgroundBlurSlot, properties {padding=<ltrb 0 0 0 0>}}
  }
  Blur_SkipSignIn {
   widgetClassPath=&BackgroundBlur
   bInherited=true
   bIsVariable=true
   parent=Switcher_Main
   widget=Blur_SkipSignIn
   properties {verticalAlignment=VAlign_Center}
   slot {class=WidgetSwitcherSlot, properties {horizontalAlignment=HAlign_Center}}
  }
  VerticalBoxSkipSignIn {
   widgetClassPath=&VerticalBox
   parent=Blur_SkipSignIn
   widget=VerticalBoxSkipSignIn
   slot {class=BackgroundBlurSlot, properties {verticalAlignment=VAlign_Center, padding=<ltrb 0 0 0 0>}}
  }
  TitleText {
   widgetClassPath=&CommonTextBlock
   parent=VerticalBoxSkipSignIn
   widget=TitleText
   properties {
    applyLineHeightToBottomLine=false
    justification=Center
    text=ARE YOU SURE?
    colorAndOpacity {colorUseRule=UseColor_Specified, specifiedColor=<rgba 0.9216 1 0 1>}
    font {
     bForceMonospaced=false
     bMaterialIsStencil=false
     fontMaterial=null
     fontObject=&Burbank
     letterSpacing=0
     monospacedWidth=1
     size=48
     skewAmount=0
     typefaceFontName=Burbank-Black
     outlineSettings {
      bApplyOutlineToDropShadows=false
      bMiteredCorners=false
      bSeparateFillAlpha=false
      outlineMaterial=null
      outlineSize=0
      outlineColor=<rgba 0 0 0 1>
     }
    }
    shadowOffset=<0 0>
   }
   slot {class=VerticalBoxSlot, properties {verticalAlignment=VAlign_Top}}
  }
  SkipSigninText {
   widgetClassPath=&CommonTextBlock
   parent=VerticalBoxSkipSignIn
   widget=SkipSigninText
   properties {
    justification=Center
    text=You will not be able to link your account in the future if you proceed without signing up or logging in.
    wrapTextAt=800
    font {
     bForceMonospaced=false
     bMaterialIsStencil=false
     fontMaterial=null
     fontObject=&Burbank
     letterSpacing=0
     monospacedWidth=1
     size=32
     skewAmount=0
     typefaceFontName=Burbank-Black
     outlineSettings {
      bApplyOutlineToDropShadows=false
      bMiteredCorners=false
      bSeparateFillAlpha=false
      outlineMaterial=null
      outlineSize=0
      outlineColor=<rgba 0 0 0 1>
     }
    }
    shadowOffset=<0 0>
   }
   slot {class=VerticalBoxSlot, properties {horizontalAlignment=HAlign_Center, verticalAlignment=VAlign_Top, padding=<ltrb 0 0 0 24>}}
  }
  VerticalBox_2 {
   widgetClassPath=&VerticalBox
   parent=VerticalBoxSkipSignIn
   widget=VerticalBox_2
   slot {class=VerticalBoxSlot, properties {horizontalAlignment=HAlign_Center, verticalAlignment=VAlign_Top}}
  }
  Button_SkipGoBackLinking {
   widgetClassPath=&IconTextButton
   bInherited=true
   bIsVariable=true
   parent=VerticalBox_2
   widget=Button_SkipGoBackLinking
   properties {
    buttonText=Go Back
    minWidth=440
    onButtonBaseClicked=(&self:WidgetTree.Button_SkipGoBackLinking.HandleOnButtonClicked)
    onButtonBaseDoubleClicked=(&self:WidgetTree.Button_SkipGoBackLinking.HandleOnButtonDoubleClicked)
    onButtonBaseHovered=(&self:WidgetTree.Button_SkipGoBackLinking.HandleOnButtonHovered)
    onButtonBaseUnhovered=(&self:WidgetTree.Button_SkipGoBackLinking.HandleOnButtonUnhovered)
    onSelectedChangedBase=(&self:WidgetTree.Button_SkipGoBackLinking.HandleOnSelectedChanged)
    style=&ButtonStyle_Feature_L_Yellow_C
   }
   slot {class=VerticalBoxSlot, properties {horizontalAlignment=HAlign_Center, padding=<ltrb 10 10 10 10>}}
  }
  Switcher_NoThanks {
   widgetClassPath=&CommonWidgetSwitcherLegacy
   bInherited=true
   bIsVariable=true
   parent=VerticalBoxSkipSignIn
   widget=Switcher_NoThanks
   slot {class=VerticalBoxSlot, properties {horizontalAlignment=HAlign_Center, verticalAlignment=VAlign_Top, padding=<ltrb 0 24 0 0>}}
  }
  Throbber_LoginDelay {
   widgetClassPath=/Script/UMG.CircularThrobber
   bInherited=true
   bIsVariable=true
   parent=Switcher_NoThanks
   widget=Throbber_LoginDelay
   properties {radius=32}
   slot {class=WidgetSwitcherSlot, properties {padding=<ltrb 0 20 0 0>}}
  }
  Button_skipSignInNoThanks {
   widgetClassPath=&IconTextButton
   bInherited=true
   bIsVariable=true
   parent=Switcher_NoThanks
   widget=Button_skipSignInNoThanks
   properties {
    buttonText="No Thanks, Play Fortnite"
    minWidth=600
    onButtonBaseClicked=(&self:WidgetTree.Button_skipSignInNoThanks.HandleOnButtonClicked)
    onButtonBaseDoubleClicked=(&self:WidgetTree.Button_skipSignInNoThanks.HandleOnButtonDoubleClicked)
    onButtonBaseHovered=(&self:WidgetTree.Button_skipSignInNoThanks.HandleOnButtonHovered)
    onButtonBaseUnhovered=(&self:WidgetTree.Button_skipSignInNoThanks.HandleOnButtonUnhovered)
    onSelectedChangedBase=(&self:WidgetTree.Button_skipSignInNoThanks.HandleOnSelectedChanged)
    style=/FortUILibrary/Foundation/Buttons/ButtonStyles/ButtonStyle_Feature_L.ButtonStyle_Feature_L_C
   }
   slot {class=WidgetSwitcherSlot, properties {horizontalAlignment=HAlign_Center, verticalAlignment=VAlign_Top, padding=<ltrb 10 10 10 10>}}
  }
 ]
}
