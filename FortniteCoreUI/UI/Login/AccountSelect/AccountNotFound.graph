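# Login screen widget shown when the chosen login method has no linked Epic account, derived from the native FortAccountNotFound class.
# A centered header and rich-text explanation sit above two inherited buttons: try another login method, or link the account at fortnite.com.
# The web button hosts an external-link icon in its named slot.
# The graph only fires an element-interacted event on each click, leaving any navigation outside this Blueprint.
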WidgetBlueprint /FortniteCoreUI/UI/Login/AccountSelect/AccountNotFound
graphs [EventGraph {nodes [
   n1 = "On Button Clicked (Button_Back)"!() -> (OutputDelegate, then, Button:Common Button Legacy Object Reference) {is=title K2Node_ComponentBoundEvent}
   > Fire Event Element Interacted(Element=@n1.Button)
   n2 = "On Button Clicked (Button_Web)"!() -> (OutputDelegate, then, Button:Common Button Legacy Object Reference) {is=title K2Node_ComponentBoundEvent}
   > Fire Event Element Interacted(Element=@n2.Button)
  ]}]
variables []
widgets {
 info {
  inheritedWidgetCount=3
  namedSlotCount=1
  parentClass=/Script/FortniteUI.FortAccountNotFound
  rootWidgetClass=&Border
  widgetCount=11
 }
 widgets [
  Border_0 {
   widgetClassPath=&Border
   widget=Border_0
   properties {
    horizontalAlignment=HAlign_Center
    visibility=SelfHitTestInvisible
    background {
     bIsDynamicallyLoaded=false
     drawAs=NoDrawType
     imageType=NoImage
     mirroring=NoMirror
     resourceName=None
     resourceObject=null
     tiling=NoTile
     imageSize=<32 32>
     margin=<ltrb 0 0 0 0>
     outlineSettings {
      bUseBrushTransparency=false
      roundingType=HalfHeightRadius
      width=0
      color {colorUseRule=UseColor_Specified, specifiedColor=<rgba 0 0 0 0>}
      cornerRadii=<0 0 0 0>
     }
     tintColor {colorUseRule=UseColor_Specified, specifiedColor=<rgba 1 1 1 1>}
     uVRegion {bIsValid=false, max=<0 0>, min=<0 0>}
    }
    padding=<ltrb 0 0 0 0>
   }
  }
  SizeBox_0 {
   widgetClassPath=&SizeBox
   parent=Border_0
   widget=SizeBox_0
   properties {bOverride_MinDesiredWidth=true, minDesiredWidth=1300}
   slot {class=BorderSlot, properties {horizontalAlignment=HAlign_Center, padding=<ltrb 0 0 0 0>}}
  }
  VerticalBox_0 {widgetClassPath=&VerticalBox, parent=SizeBox_0, widget=VerticalBox_0, slot {class=SizeBoxSlot}}
  Overlay_0 {
   widgetClassPath=&Overlay
   parent=VerticalBox_0
   widget=Overlay_0
   slot {class=VerticalBoxSlot, properties {horizontalAlignment=HAlign_Center, verticalAlignment=VAlign_Center, size {sizeRule=Fill, value=1}}}
  }
  VerticalBox_1 {widgetClassPath=&VerticalBox, parent=Overlay_0, widget=VerticalBox_1, slot {class=OverlaySlot}}
  CommonTextBlock_0 {
   widgetClassPath=&CommonTextBlock
   parent=VerticalBox_1
   widget=CommonTextBlock_0
   properties {
    style=/FortUILibrary/Styles/Typography/Foundation/TextStyle-Header-L.TextStyle-Header-L_C
    text=EPIC ACCOUNT NOT FOUND
    font {
     bForceMonospaced=false
     bMaterialIsStencil=false
     fontMaterial=null
     fontObject=&Burbank
     letterSpacing=0
     monospacedWidth=1
     size=48
     skewAmount=0
     typefaceFontName=Burbank-Black
     outlineSettings {
      bApplyOutlineToDropShadows=false
      bMiteredCorners=false
      bSeparateFillAlpha=false
      outlineMaterial=null
      outlineSize=0
      outlineColor=<rgba 0 0 0 1>
     }
    }
    shadowOffset=<0 0>
   }
   slot {class=VerticalBoxSlot, properties {horizontalAlignment=HAlign_Center}}
  }
  Text_Desc {
   widgetClassPath=/Script/UMG.RichTextBlock
   bInherited=true
   bIsVariable=true
   parent=VerticalBox_1
   widget=Text_Desc
   properties {
    autoWrapText=true
    justification=Center
    text="An <stronger>Epic Games</> account is required to play <stronger>Fortnite</>, but your login method is not linked to an existing <stronger>Epic Games</> account.\r\n\r\nPlease try a different login method, or link your account at the <stronger>Fortnite</> website."
    textStyleSet=/FortniteCoreUI/UI/Frontend/FriendCodes/FriendCodeRichTextStyles
    wrapTextAt=900
   }
   slot {class=VerticalBoxSlot, properties {padding=<ltrb 20 20 20 20>}}
  }
  Button_Back {
   widgetClassPath=&IconTextButton
   bInherited=true
   bIsVariable=true
   parent=VerticalBox_1
   widget=Button_Back
   properties {
    buttonText=TRY A DIFFERENT LOGIN METHOD
    contentAlignment=HAlign_Center
    minHeight=100
    onButtonBaseClicked=(&self:WidgetTree.Button_Back.HandleOnButtonClicked)
    onButtonBaseDoubleClicked=(&self:WidgetTree.Button_Back.HandleOnButtonDoubleClicked)
    onButtonBaseHovered=(&self:WidgetTree.Button_Back.HandleOnButtonHovered)
    onButtonBaseUnhovered=(&self:WidgetTree.Button_Back.HandleOnButtonUnhovered)
    onSelectedChangedBase=(&self:WidgetTree.Button_Back.HandleOnSelectedChanged)
    style=/FortUILibrary/Foundation/Buttons/ButtonStyles/Athena/Athena_ButtonStyle_AngledBlueMenuButton.Athena_ButtonStyle_AngledBlueMenuButton_C
   }
   slot {class=VerticalBoxSlot, properties {padding=<ltrb 20 20 20 20>}}
  }
  Button_Web {
   widgetClassPath=&IconTextButton
   bInherited=true
   bIsVariable=true
   parent=VerticalBox_1
   widget=Button_Web
   properties {
    buttonText=LINK ACCOUNT AT fortnite.com
    contentAlignment=HAlign_Center
    minHeight=100
    onButtonBaseClicked=(&self:WidgetTree.Button_Web.HandleOnButtonClicked)
    onButtonBaseDoubleClicked=(&self:WidgetTree.Button_Web.HandleOnButtonDoubleClicked)
    onButtonBaseHovered=(&self:WidgetTree.Button_Web.HandleOnButtonHovered)
    onButtonBaseUnhovered=(&self:WidgetTree.Button_Web.HandleOnButtonUnhovered)
    onSelectedChangedBase=(&self:WidgetTree.Button_Web.HandleOnSelectedChanged)
    style=&Athena_ButtonStyle_AngledBlueMenuButton_C
   }
   slot {class=VerticalBoxSlot, properties {padding=<ltrb 20 20 20 20>}}
  }
  Border_1 {
   widgetClassPath=&Border
   namedSlotHost=Button_Web
   widget=Border_1
   properties {
    horizontalAlignment=HAlign_Left
    verticalAlignment=VAlign_Center
    visibility=SelfHitTestInvisible
    background {
     bIsDynamicallyLoaded=false
     drawAs=NoDrawType
     imageType=NoImage
     mirroring=NoMirror
     resourceName=None
     resourceObject=null
     tiling=NoTile
     imageSize=<32 32>
     margin=<ltrb 0 0 0 0>
     outlineSettings {
      bUseBrushTransparency=false
      roundingType=HalfHeightRadius
      width=0
      color {colorUseRule=UseColor_Specified, specifiedColor=<rgba 0 0 0 0>}
      cornerRadii=<0 0 0 0>
     }
     tintColor {colorUseRule=UseColor_Specified, specifiedColor=<rgba 1 1 1 1>}
     uVRegion {bIsValid=false, max=<0 0>, min=<0 0>}
    }
    padding=<ltrb 0 0 0 0>
   }
  }
  ImageExternalLink {
   widgetClassPath=&Image
   parent=Border_1
   widget=ImageExternalLink
   properties {brush {
     bIsDynamicallyLoaded=false
     drawAs=Image
     imageType=NoImage
     mirroring=NoMirror
     resourceName=None
     resourceObject=/FortniteCoreUI/UI/Login/Art/M-Icon-ExternalLink-64
     tiling=NoTile
     imageSize=<75 75>
     margin=<ltrb 0 0 0 0>
     outlineSettings {
      bUseBrushTransparency=false
      roundingType=HalfHeightRadius
      width=0
      color {colorUseRule=UseColor_Specified, specifiedColor=<rgba 0 0 0 0>}
      cornerRadii=<0 0 0 0>
     }
     tintColor {colorUseRule=UseColor_Specified, specifiedColor=<rgba 1 1 1 1>}
     uVRegion {bIsValid=false, max=<0 0>, min=<0 0>}
    }}
   slot {class=BorderSlot, properties {horizontalAlignment=HAlign_Left, verticalAlignment=VAlign_Center, padding=<ltrb 0 0 0 0>}}
  }
 ]
}
