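# Login screen widget where players pick how they log in to Fortnite (Epic, Facebook, Google, PlayStation, Xbox Live, Nintendo Switch) or create an Epic account.
# Extends FortLoginCredentialSelect, laying out its inherited buttons under a two-prompt switcher, with a console trademark panel.
# Click handlers only fire Event Element Interacted with the button; no login or switcher logic is in this graph.
# Construct gives the Switch icon a dynamic media-player-button material with PlayButton set.
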WidgetBlueprint /FortniteCoreUI/UI/Login/AccountSelect/LoginCredentialSelect
graphs [EventGraph {nodes [
   n1 = "On Button Clicked (Button_Epic)"!() -> (OutputDelegate, then, Button:Common Button Legacy Object Reference) {is=title K2Node_ComponentBoundEvent}
   > Fire Event Element Interacted(Element=@n1.Button)
   n2 = "On Button Clicked (Button_Facebook)"!() -> (OutputDelegate, then, Button:Common Button Legacy Object Reference) {is=title K2Node_ComponentBoundEvent}
   > Fire Event Element Interacted(Element=@n2.Button)
   n3 = "On Button Clicked (Button_Google)"!() -> (OutputDelegate, then, Button:Common Button Legacy Object Reference) {is=title K2Node_ComponentBoundEvent}
   > Fire Event Element Interacted(Element=@n3.Button)
   n4 = "On Button Clicked (Button_PS)"!() -> (OutputDelegate, then, Button:Common Button Legacy Object Reference) {is=title K2Node_ComponentBoundEvent}
   > Fire Event Element Interacted(Element=@n4.Button)
   n5 = "On Button Clicked (Button_XB)"!() -> (OutputDelegate, then, Button:Common Button Legacy Object Reference) {is=title K2Node_ComponentBoundEvent}
   > Fire Event Element Interacted(Element=@n5.Button)
   n6 = "On Button Clicked (Button_Erebus)"!() -> (OutputDelegate, then, Button:Common Button Legacy Object Reference) {is=title K2Node_ComponentBoundEvent}
   > Fire Event Element Interacted(Element=@n6.Button)
   n7 = "On Button Clicked (Button_CreateAccount)"!() -> (OutputDelegate, then, Button:Common Button Legacy Object Reference) {is=title K2Node_ComponentBoundEvent}
   > Fire Event Element Interacted(Element=@n7.Button)
   Event Construct()
   > n8 = CreateDynamicMaterialInstance!(execute, Parent=/FortniteCoreUI/UI/Materials/M_MediaPlayerButton_Base, OptionalName=None, CreationFlags=None) -> (then, ReturnValue)
   > SetScalarParameterValue(self=@n8, ParameterName=PlayButton, Value=1)
   > SetBrushFromMaterial(self=Get Image_5() -> (Image_5:Image Object Reference), Material=@n8)
  ]}]
variables []
widgets {
 info {
  inheritedWidgetCount=7
  namedSlotCount=7
  parentClass=/Script/FortniteUI.FortLoginCredentialSelect
  rootWidgetClass=&SafeZone
  widgetCount=55
 }
 widgets [
  SafeZone_0 {widgetClassPath=&SafeZone, bIsVariable=true, widget=SafeZone_0, properties {padTop=false}}
  Overlay_0 {widgetClassPath=&Overlay, parent=SafeZone_0, widget=Overlay_0, slot {class=SafeZoneSlot}}
  VerticalBox {
   widgetClassPath=&VerticalBox
   parent=Overlay_0
   widget=VerticalBox
   slot {class=OverlaySlot, properties {horizontalAlignment=HAlign_Fill, verticalAlignment=VAlign_Center, padding=<ltrb 64 0 64 0>}}
  }
  CommonWidgetSwitcher_0 {
   widgetClassPath=&CommonWidgetSwitcherLegacy
   bIsVariable=true
   parent=VerticalBox
   widget=CommonWidgetSwitcher_0
   slot {class=VerticalBoxSlot, properties {verticalAlignment=VAlign_Center}}
  }
  CommonTextBlock_1 {
   widgetClassPath=&CommonTextBlock
   parent=CommonWidgetSwitcher_0
   widget=CommonTextBlock_1
   properties {
    autoWrapText=true
    style=/FortUILibrary/Styles/Typography/Foundation/TextStyle-Base-M.TextStyle-Base-M_C
    text=WE'LL FIND YOUR ACCOUNT. TELL US THE WAY YOU LOG IN TO FORTNITE.
    font {
     bForceMonospaced=false
     bMaterialIsStencil=false
     fontMaterial=null
     fontObject=&NotoSans
     letterSpacing=0
     monospacedWidth=1
     size=24
     skewAmount=0
     typefaceFontName=NotoSans-Regular
     outlineSettings {
      bApplyOutlineToDropShadows=false
      bMiteredCorners=false
      bSeparateFillAlpha=false
      outlineMaterial=null
      outlineSize=0
      outlineColor=<rgba 0 0 0 1>
     }
    }
    shadowOffset=<0 0>
   }
   slot {class=WidgetSwitcherSlot}
  }
  CommonTextBlock_8 {
   widgetClassPath=&CommonTextBlock
   parent=CommonWidgetSwitcher_0
   widget=CommonTextBlock_8
   properties {
    autoWrapText=true
    style=/FortUILibrary/Styles/Typography/Foundation/TextStyle-Base-M-B.TextStyle-Base-M-B_C
    text=Log in to find your Fortnite progress and purchases. Select the log in method you use most often.
    font {
     bForceMonospaced=false
     bMaterialIsStencil=false
     fontMaterial=null
     fontObject=&NotoSans
     letterSpacing=0
     monospacedWidth=1
     size=24
     skewAmount=0
     typefaceFontName=NotoSans-Bold
     outlineSettings {
      bApplyOutlineToDropShadows=false
      bMiteredCorners=false
      bSeparateFillAlpha=false
      outlineMaterial=null
      outlineSize=0
      outlineColor=<rgba 0 0 0 1>
     }
    }
    shadowOffset=<0 0>
   }
   slot {class=WidgetSwitcherSlot}
  }
  WrapBox_0 {
   widgetClassPath=&WrapBox
   parent=VerticalBox
   widget=WrapBox_0
   properties {horizontalAlignment=HAlign_Center}
   slot {class=VerticalBoxSlot}
  }
  Button_Epic {
   widgetClassPath=/FortUILibrary/Foundation/Buttons/SkewButton
   bInherited=true
   bIsVariable=true
   parent=WrapBox_0
   widget=Button_Epic
   properties {
    bIsFocusable=true
    onButtonBaseClicked=(&self:WidgetTree.Button_Epic.HandleOnButtonClicked)
    onButtonBaseDoubleClicked=(&self:WidgetTree.Button_Epic.HandleOnButtonDoubleClicked)
    onButtonBaseHovered=(&self:WidgetTree.Button_Epic.HandleOnButtonHovered)
    onButtonBaseUnhovered=(&self:WidgetTree.Button_Epic.HandleOnButtonUnhovered)
    onSelectedChangedBase=(&self:WidgetTree.Button_Epic.HandleOnSelectedChanged)
    skew=false
    buttonColor=<rgba 0.0194 0.0185 0.0152 1>
   }
   slot {class=WrapBoxSlot, properties {horizontalAlignment=HAlign_Center, verticalAlignment=VAlign_Center}}
  }
  CommonBorder_0 {widgetClassPath=&CommonBorder, namedSlotHost=Button_Epic, widget=CommonBorder_0, properties {padding=<ltrb 0 35 0 65>}}
  SizeBox_2 {
   widgetClassPath=&SizeBox
   parent=CommonBorder_0
   widget=SizeBox_2
   properties {bOverride_MinDesiredWidth=true, minDesiredWidth=200, widthOverride=200}
   slot {class=BorderSlot, properties {padding=<ltrb 0 35 0 65>}}
  }
  VerticalBox_0 {widgetClassPath=&VerticalBox, parent=SizeBox_2, widget=VerticalBox_0, slot {class=SizeBoxSlot}}
  Image_48 {
   widgetClassPath=&Image
   bIsVariable=true
   parent=VerticalBox_0
   widget=Image_48
   properties {brush {
     bIsDynamicallyLoaded=false
     drawAs=Image
     imageType=NoImage
     mirroring=NoMirror
     resourceName=None
     resourceObject=/FortniteCoreUI/UI/Login/Art/M_TempIcon_Login_Epic
     tiling=NoTile
     imageSize=<170 170>
     margin=<ltrb 0 0 0 0>
     outlineSettings {
      bUseBrushTransparency=false
      roundingType=HalfHeightRadius
      width=0
      color {colorUseRule=UseColor_Specified, specifiedColor=<rgba 0 0 0 0>}
      cornerRadii=<0 0 0 0>
     }
     tintColor {colorUseRule=UseColor_Specified, specifiedColor=<rgba 1 1 1 1>}
     uVRegion {bIsValid=false, max=<0 0>, min=<0 0>}
    }}
   slot {class=VerticalBoxSlot, properties {horizontalAlignment=HAlign_Center, verticalAlignment=VAlign_Center, padding=<ltrb 0 0 0 -12>}}
  }
  CommonTextBlock_3 {
   widgetClassPath=&CommonTextBlock
   parent=VerticalBox_0
   widget=CommonTextBlock_3
   properties {
    justification=Center
    style=&TextStyle-Base-M-B_C
    text=Epic Games
    font {
     bForceMonospaced=false
     bMaterialIsStencil=false
     fontMaterial=null
     fontObject=&NotoSans
     letterSpacing=0
     monospacedWidth=1
     size=24
     skewAmount=0
     typefaceFontName=NotoSans-Bold
     outlineSettings {
      bApplyOutlineToDropShadows=false
      bMiteredCorners=false
      bSeparateFillAlpha=false
      outlineMaterial=null
      outlineSize=0
      outlineColor=<rgba 0 0 0 1>
     }
    }
    shadowOffset=<0 0>
   }
   slot {class=VerticalBoxSlot}
  }
  Button_Facebook {
   widgetClassPath=&SkewButton
   bInherited=true
   bIsVariable=true
   parent=WrapBox_0
   widget=Button_Facebook
   properties {
    bIsFocusable=true
    onButtonBaseClicked=(&self:WidgetTree.Button_Facebook.HandleOnButtonClicked)
    onButtonBaseDoubleClicked=(&self:WidgetTree.Button_Facebook.HandleOnButtonDoubleClicked)
    onButtonBaseHovered=(&self:WidgetTree.Button_Facebook.HandleOnButtonHovered)
    onButtonBaseUnhovered=(&self:WidgetTree.Button_Facebook.HandleOnButtonUnhovered)
    onSelectedChangedBase=(&self:WidgetTree.Button_Facebook.HandleOnSelectedChanged)
    skew=false
    buttonColor=<rgba 0.0409 0.0908 0.3467 1>
   }
   slot {class=WrapBoxSlot, properties {horizontalAlignment=HAlign_Center, verticalAlignment=VAlign_Center}}
  }
  CommonBorder_1 {widgetClassPath=&CommonBorder, namedSlotHost=Button_Facebook, widget=CommonBorder_1, properties {padding=<ltrb 0 35 0 65>}}
  SizeBox_1 {
   widgetClassPath=&SizeBox
   parent=CommonBorder_1
   widget=SizeBox_1
   properties {bOverride_MinDesiredWidth=true, minDesiredWidth=200, widthOverride=200}
   slot {class=BorderSlot, properties {padding=<ltrb 0 35 0 65>}}
  }
  VerticalBox_1 {widgetClassPath=&VerticalBox, parent=SizeBox_1, widget=VerticalBox_1, slot {class=SizeBoxSlot}}
  Image_0 {
   widgetClassPath=&Image
   bIsVariable=true
   parent=VerticalBox_1
   widget=Image_0
   properties {
    brush {
     bIsDynamicallyLoaded=false
     drawAs=Image
     imageType=NoImage
     mirroring=NoMirror
     resourceName=None
     resourceObject=/FortniteCoreUI/UI/Login/Art/M_TempIcon_Login_Facebook
     tiling=NoTile
     imageSize=<128 128>
     margin=<ltrb 0 0 0 0>
     outlineSettings {
      bUseBrushTransparency=false
      roundingType=HalfHeightRadius
      width=0
      color {colorUseRule=UseColor_Specified, specifiedColor=<rgba 0 0 0 0>}
      cornerRadii=<0 0 0 0>
     }
     tintColor {colorUseRule=UseColor_Specified, specifiedColor=<rgba 1 1 1 1>}
     uVRegion {bIsValid=false, max=<0 0>, min=<0 0>}
    }
    renderTransform {angle=0, scale=<1 1>, shear=<0 0>, translation=<-25 0>}
   }
   slot {class=VerticalBoxSlot, properties {horizontalAlignment=HAlign_Center, verticalAlignment=VAlign_Center, padding=<ltrb 0 12 0 16>}}
  }
  CommonTextBlock_4 {
   widgetClassPath=&CommonTextBlock
   parent=VerticalBox_1
   widget=CommonTextBlock_4
   properties {
    justification=Center
    style=&TextStyle-Base-M-B_C
    text=Facebook
    font {
     bForceMonospaced=false
     bMaterialIsStencil=false
     fontMaterial=null
     fontObject=&NotoSans
     letterSpacing=0
     monospacedWidth=1
     size=24
     skewAmount=0
     typefaceFontName=NotoSans-Bold
     outlineSettings {
      bApplyOutlineToDropShadows=false
      bMiteredCorners=false
      bSeparateFillAlpha=false
      outlineMaterial=null
      outlineSize=0
      outlineColor=<rgba 0 0 0 1>
     }
    }
    shadowOffset=<0 0>
   }
   slot {class=VerticalBoxSlot}
  }
  Button_Google {
   widgetClassPath=&SkewButton
   bInherited=true
   bIsVariable=true
   parent=WrapBox_0
   widget=Button_Google
   properties {
    bIsFocusable=true
    onButtonBaseClicked=(&self:WidgetTree.Button_Google.HandleOnButtonClicked)
    onButtonBaseDoubleClicked=(&self:WidgetTree.Button_Google.HandleOnButtonDoubleClicked)
    onButtonBaseHovered=(&self:WidgetTree.Button_Google.HandleOnButtonHovered)
    onButtonBaseUnhovered=(&self:WidgetTree.Button_Google.HandleOnButtonUnhovered)
    onSelectedChangedBase=(&self:WidgetTree.Button_Google.HandleOnSelectedChanged)
    skew=false
    buttonColor=<rgba 1 1 1 1>
   }
   slot {class=WrapBoxSlot, properties {horizontalAlignment=HAlign_Center, verticalAlignment=VAlign_Center}}
  }
  CommonBorder_2 {widgetClassPath=&CommonBorder, namedSlotHost=Button_Google, widget=CommonBorder_2, properties {padding=<ltrb 0 35 0 65>}}
  SizeBox_3 {
   widgetClassPath=&SizeBox
   parent=CommonBorder_2
   widget=SizeBox_3
   properties {bOverride_MinDesiredWidth=true, minDesiredWidth=200, widthOverride=200}
   slot {class=BorderSlot, properties {padding=<ltrb 0 35 0 65>}}
  }
  VerticalBox_2 {widgetClassPath=&VerticalBox, parent=SizeBox_3, widget=VerticalBox_2, slot {class=SizeBoxSlot}}
  Image_1 {
   widgetClassPath=&Image
   bIsVariable=true
   parent=VerticalBox_2
   widget=Image_1
   properties {brush {
     bIsDynamicallyLoaded=false
     drawAs=Image
     imageType=NoImage
     mirroring=NoMirror
     resourceName=None
     resourceObject=/FortniteCoreUI/UI/Login/Art/M_TempIcon_Login_Google
     tiling=NoTile
     imageSize=<156 156>
     margin=<ltrb 0 0 0 0>
     outlineSettings {
      bUseBrushTransparency=false
      roundingType=HalfHeightRadius
      width=0
      color {colorUseRule=UseColor_Specified, specifiedColor=<rgba 0 0 0 0>}
      cornerRadii=<0 0 0 0>
     }
     tintColor {colorUseRule=UseColor_Specified, specifiedColor=<rgba 1 1 1 1>}
     uVRegion {bIsValid=false, max=<0 0>, min=<0 0>}
    }}
   slot {class=VerticalBoxSlot, properties {horizontalAlignment=HAlign_Center, verticalAlignment=VAlign_Center}}
  }
  CommonTextBlock_5 {
   widgetClassPath=&CommonTextBlock
   parent=VerticalBox_2
   widget=CommonTextBlock_5
   properties {
    justification=Center
    style=/FortUILibrary/Styles/Typography/Foundation/TextStyle-Base-M-B_Black.TextStyle-Base-M-B_Black_C
    text=Google
    colorAndOpacity {colorUseRule=UseColor_Specified, specifiedColor=<rgba 0 0 0 1>}
    font {
     bForceMonospaced=false
     bMaterialIsStencil=false
     fontMaterial=null
     fontObject=&NotoSans
     letterSpacing=0
     monospacedWidth=1
     size=24
     skewAmount=0
     typefaceFontName=NotoSans-Bold
     outlineSettings {
      bApplyOutlineToDropShadows=false
      bMiteredCorners=false
      bSeparateFillAlpha=false
      outlineMaterial=null
      outlineSize=0
      outlineColor=<rgba 0 0 0 1>
     }
    }
    shadowOffset=<0 0>
   }
   slot {class=VerticalBoxSlot}
  }
  Button_PS {
   widgetClassPath=&SkewButton
   bInherited=true
   bIsVariable=true
   parent=WrapBox_0
   widget=Button_PS
   properties {
    bIsFocusable=true
    onButtonBaseClicked=(&self:WidgetTree.Button_PS.HandleOnButtonClicked)
    onButtonBaseDoubleClicked=(&self:WidgetTree.Button_PS.HandleOnButtonDoubleClicked)
    onButtonBaseHovered=(&self:WidgetTree.Button_PS.HandleOnButtonHovered)
    onButtonBaseUnhovered=(&self:WidgetTree.Button_PS.HandleOnButtonUnhovered)
    onSelectedChangedBase=(&self:WidgetTree.Button_PS.HandleOnSelectedChanged)
    skew=false
    buttonColor=<rgba 0.008 0.0595 0.2502 1>
   }
   slot {class=WrapBoxSlot, properties {horizontalAlignment=HAlign_Center, verticalAlignment=VAlign_Center}}
  }
  CommonBorder_3 {widgetClassPath=&CommonBorder, namedSlotHost=Button_PS, widget=CommonBorder_3, properties {padding=<ltrb 0 35 0 65>}}
  SizeBox_4 {
   widgetClassPath=&SizeBox
   parent=CommonBorder_3
   widget=SizeBox_4
   properties {bOverride_MinDesiredWidth=true, minDesiredWidth=200, widthOverride=200}
   slot {class=BorderSlot, properties {padding=<ltrb 0 35 0 65>}}
  }
  VerticalBox_3 {widgetClassPath=&VerticalBox, parent=SizeBox_4, widget=VerticalBox_3, slot {class=SizeBoxSlot}}
  Image_2 {
   widgetClassPath=&Image
   bIsVariable=true
   parent=VerticalBox_3
   widget=Image_2
   properties {brush {
     bIsDynamicallyLoaded=false
     drawAs=Image
     imageType=NoImage
     mirroring=NoMirror
     resourceName=None
     resourceObject=/FortniteCoreUI/UI/Login/Art/M_TempIcon_Login_PS4
     tiling=NoTile
     imageSize=<156 156>
     margin=<ltrb 0 0 0 0>
     outlineSettings {
      bUseBrushTransparency=false
      roundingType=HalfHeightRadius
      width=0
      color {colorUseRule=UseColor_Specified, specifiedColor=<rgba 0 0 0 0>}
      cornerRadii=<0 0 0 0>
     }
     tintColor {colorUseRule=UseColor_Specified, specifiedColor=<rgba 1 1 1 1>}
     uVRegion {bIsValid=false, max=<0 0>, min=<0 0>}
    }}
   slot {class=VerticalBoxSlot, properties {horizontalAlignment=HAlign_Center, verticalAlignment=VAlign_Center}}
  }
  CommonTextBlock_6 {
   widgetClassPath=&CommonTextBlock
   parent=VerticalBox_3
   widget=CommonTextBlock_6
   properties {
    justification=Center
    style=&TextStyle-Base-M-B_C
    text="PlayStation®"
    font {
     bForceMonospaced=false
     bMaterialIsStencil=false
     fontMaterial=null
     fontObject=&NotoSans
     letterSpacing=0
     monospacedWidth=1
     size=24
     skewAmount=0
     typefaceFontName=NotoSans-Bold
     outlineSettings {
      bApplyOutlineToDropShadows=false
      bMiteredCorners=false
      bSeparateFillAlpha=false
      outlineMaterial=null
      outlineSize=0
      outlineColor=<rgba 0 0 0 1>
     }
    }
    shadowOffset=<0 0>
   }
   slot {class=VerticalBoxSlot}
  }
  Button_XB {
   widgetClassPath=&SkewButton
   bInherited=true
   bIsVariable=true
   parent=WrapBox_0
   widget=Button_XB
   properties {
    bIsFocusable=true
    onButtonBaseClicked=(&self:WidgetTree.Button_XB.HandleOnButtonClicked)
    onButtonBaseDoubleClicked=(&self:WidgetTree.Button_XB.HandleOnButtonDoubleClicked)
    onButtonBaseHovered=(&self:WidgetTree.Button_XB.HandleOnButtonHovered)
    onButtonBaseUnhovered=(&self:WidgetTree.Button_XB.HandleOnButtonUnhovered)
    onSelectedChangedBase=(&self:WidgetTree.Button_XB.HandleOnSelectedChanged)
    skew=false
    buttonColor=<rgba 0.0273 0.4072 0.0052 1>
   }
   slot {class=WrapBoxSlot, properties {horizontalAlignment=HAlign_Center, verticalAlignment=VAlign_Center}}
  }
  CommonBorder_4 {widgetClassPath=&CommonBorder, namedSlotHost=Button_XB, widget=CommonBorder_4, properties {padding=<ltrb 0 35 0 65>}}
  SizeBox_5 {
   widgetClassPath=&SizeBox
   parent=CommonBorder_4
   widget=SizeBox_5
   properties {bOverride_MinDesiredWidth=true, minDesiredWidth=200, widthOverride=200}
   slot {class=BorderSlot, properties {padding=<ltrb 0 35 0 65>}}
  }
  VerticalBox_4 {widgetClassPath=&VerticalBox, parent=SizeBox_5, widget=VerticalBox_4, slot {class=SizeBoxSlot}}
  Image_3 {
   widgetClassPath=&Image
   bIsVariable=true
   parent=VerticalBox_4
   widget=Image_3
   properties {brush {
     bIsDynamicallyLoaded=false
     drawAs=Image
     imageType=NoImage
     mirroring=NoMirror
     resourceName=None
     resourceObject=/FortniteCoreUI/UI/Login/Art/M_TempIcon_Login_XBox
     tiling=NoTile
     imageSize=<156 156>
     margin=<ltrb 0 0 0 0>
     outlineSettings {
      bUseBrushTransparency=false
      roundingType=HalfHeightRadius
      width=0
      color {colorUseRule=UseColor_Specified, specifiedColor=<rgba 0 0 0 0>}
      cornerRadii=<0 0 0 0>
     }
     tintColor {colorUseRule=UseColor_Specified, specifiedColor=<rgba 1 1 1 1>}
     uVRegion {bIsValid=false, max=<0 0>, min=<0 0>}
    }}
   slot {class=VerticalBoxSlot, properties {horizontalAlignment=HAlign_Center, verticalAlignment=VAlign_Center}}
  }
  CommonTextBlock_7 {
   widgetClassPath=&CommonTextBlock
   parent=VerticalBox_4
   widget=CommonTextBlock_7
   properties {
    justification=Center
    style=&TextStyle-Base-M-B_C
    text=Xbox Live
    font {
     bForceMonospaced=false
     bMaterialIsStencil=false
     fontMaterial=null
     fontObject=&NotoSans
     letterSpacing=0
     monospacedWidth=1
     size=24
     skewAmount=0
     typefaceFontName=NotoSans-Bold
     outlineSettings {
      bApplyOutlineToDropShadows=false
      bMiteredCorners=false
      bSeparateFillAlpha=false
      outlineMaterial=null
      outlineSize=0
      outlineColor=<rgba 0 0 0 1>
     }
    }
    shadowOffset=<0 0>
   }
   slot {class=VerticalBoxSlot}
  }
  Button_Erebus {
   widgetClassPath=&SkewButton
   bInherited=true
   bIsVariable=true
   parent=WrapBox_0
   widget=Button_Erebus
   properties {
    bIsFocusable=true
    onButtonBaseClicked=(&self:WidgetTree.Button_Erebus.HandleOnButtonClicked)
    onButtonBaseDoubleClicked=(&self:WidgetTree.Button_Erebus.HandleOnButtonDoubleClicked)
    onButtonBaseHovered=(&self:WidgetTree.Button_Erebus.HandleOnButtonHovered)
    onButtonBaseUnhovered=(&self:WidgetTree.Button_Erebus.HandleOnButtonUnhovered)
    onSelectedChangedBase=(&self:WidgetTree.Button_Erebus.HandleOnSelectedChanged)
    skew=false
    buttonColor=<rgba 1 0 0 1>
   }
   slot {class=WrapBoxSlot, properties {horizontalAlignment=HAlign_Center, verticalAlignment=VAlign_Center}}
  }
  CommonBorder_5 {widgetClassPath=&CommonBorder, namedSlotHost=Button_Erebus, widget=CommonBorder_5, properties {padding=<ltrb 0 35 0 65>}}
  SizeBox_6 {
   widgetClassPath=&SizeBox
   parent=CommonBorder_5
   widget=SizeBox_6
   properties {bOverride_MinDesiredWidth=true, minDesiredWidth=200, widthOverride=200}
   slot {class=BorderSlot, properties {padding=<ltrb 0 35 0 65>}}
  }
  VerticalBox_6 {widgetClassPath=&VerticalBox, parent=SizeBox_6, widget=VerticalBox_6, slot {class=SizeBoxSlot}}
  Image_5 {
   widgetClassPath=&Image
   bIsVariable=true
   parent=VerticalBox_6
   widget=Image_5
   properties {brush {
     bIsDynamicallyLoaded=false
     drawAs=Image
     imageType=NoImage
     mirroring=NoMirror
     resourceName=None
     resourceObject=&M_MediaPlayerButton_Base
     tiling=NoTile
     imageSize=<156 156>
     margin=<ltrb 0 0 0 0>
     outlineSettings {
      bUseBrushTransparency=false
      roundingType=HalfHeightRadius
      width=0
      color {colorUseRule=UseColor_Specified, specifiedColor=<rgba 0 0 0 0>}
      cornerRadii=<0 0 0 0>
     }
     tintColor {colorUseRule=UseColor_Specified, specifiedColor=<rgba 1 1 1 1>}
     uVRegion {bIsValid=false, max=<0 0>, min=<0 0>}
    }}
   slot {class=VerticalBoxSlot, properties {horizontalAlignment=HAlign_Center, verticalAlignment=VAlign_Center}}
  }
  VerticalBox_7 {widgetClassPath=&VerticalBox, parent=VerticalBox_6, widget=VerticalBox_7, slot {class=VerticalBoxSlot}}
  Erebus_Title {
   widgetClassPath=&CommonTextBlock
   bIsVariable=true
   parent=VerticalBox_7
   widget=Erebus_Title
   properties {
    justification=Center
    style=/FortUILibrary/Styles/Typography/Foundation/TextStyle-Base-XXS-B.TextStyle-Base-XXS-B_C
    text=Nintendo
    font {
     bForceMonospaced=false
     bMaterialIsStencil=false
     fontMaterial=null
     fontObject=&NotoSans
     letterSpacing=0
     monospacedWidth=1
     size=14
     skewAmount=0
     typefaceFontName=NotoSans-Bold
     outlineSettings {
      bApplyOutlineToDropShadows=false
      bMiteredCorners=false
      bSeparateFillAlpha=false
      outlineMaterial=null
      outlineSize=0
      outlineColor=<rgba 0 0 0 1>
     }
    }
    shadowOffset=<0 0>
   }
   slot {class=VerticalBoxSlot, properties {padding=<ltrb 0 0 0 -12>}}
  }
  ErebusType {
   widgetClassPath=&CommonTextBlock
   bIsVariable=true
   parent=VerticalBox_7
   widget=ErebusType
   properties {
    justification=Center
    style=&TextStyle-Base-M-B_C
    text=Switch
    font {
     bForceMonospaced=false
     bMaterialIsStencil=false
     fontMaterial=null
     fontObject=&NotoSans
     letterSpacing=0
     monospacedWidth=1
     size=24
     skewAmount=0
     typefaceFontName=NotoSans-Bold
     outlineSettings {
      bApplyOutlineToDropShadows=false
      bMiteredCorners=false
      bSeparateFillAlpha=false
      outlineMaterial=null
      outlineSize=0
      outlineColor=<rgba 0 0 0 1>
     }
    }
    shadowOffset=<0 0>
   }
   slot {class=VerticalBoxSlot, properties {padding=<ltrb 0 0 0 -10>}}
  }
  HorizontalBox_2 {
   widgetClassPath=&HorizontalBox
   parent=VerticalBox
   widget=HorizontalBox_2
   slot {class=VerticalBoxSlot, properties {horizontalAlignment=HAlign_Left, verticalAlignment=VAlign_Center}}
  }
  Button_CreateAccount {
   widgetClassPath=&IconTextButton
   bInherited=true
   bIsVariable=true
   parent=HorizontalBox_2
   widget=Button_CreateAccount
   properties {
    buttonText=Create an Epic Games Account
    contentAlignment=HAlign_Left
    minHeight=100
    onButtonBaseClicked=(&self:WidgetTree.Button_CreateAccount.HandleOnButtonClicked)
    onButtonBaseDoubleClicked=(&self:WidgetTree.Button_CreateAccount.HandleOnButtonDoubleClicked)
    onButtonBaseHovered=(&self:WidgetTree.Button_CreateAccount.HandleOnButtonHovered)
    onButtonBaseUnhovered=(&self:WidgetTree.Button_CreateAccount.HandleOnButtonUnhovered)
    onSelectedChangedBase=(&self:WidgetTree.Button_CreateAccount.HandleOnSelectedChanged)
    style=/FortUILibrary/Foundation/Buttons/ButtonStyles/ButtonStyle-TextOnlyBase_S-B.ButtonStyle-TextOnlyBase_S-B_C
    iconBrush {
     bIsDynamicallyLoaded=false
     drawAs=NoDrawType
     imageType=NoImage
     mirroring=NoMirror
     resourceName=None
     resourceObject=/FortniteCoreUI/UI/Login/Art/M-Icon-ExternalLink-64
     tiling=NoTile
     imageSize=<76 76>
     margin=<ltrb 0 0 0 0>
     outlineSettings {
      bUseBrushTransparency=false
      roundingType=HalfHeightRadius
      width=0
      color {colorUseRule=UseColor_Specified, specifiedColor=<rgba 0 0 0 0>}
      cornerRadii=<0 0 0 1>
     }
     tintColor {colorUseRule=UseColor_Specified, specifiedColor=<rgba 1 1 1 1>}
     uVRegion {bIsValid=false, max=<0 0>, min=<0 0>}
    }
   }
   slot {class=HorizontalBoxSlot, properties {horizontalAlignment=HAlign_Left, verticalAlignment=VAlign_Center}}
  }
  Border_2 {
   widgetClassPath=&Border
   namedSlotHost=Button_CreateAccount
   widget=Border_2
   properties {
    horizontalAlignment=HAlign_Left
    verticalAlignment=VAlign_Center
    background {
     bIsDynamicallyLoaded=false
     drawAs=NoDrawType
     imageType=NoImage
     mirroring=NoMirror
     resourceName=None
     resourceObject=null
     tiling=NoTile
     imageSize=<32 32>
     margin=<ltrb 0 0 0 0>
     outlineSettings {
      bUseBrushTransparency=false
      roundingType=HalfHeightRadius
      width=0
      color {colorUseRule=UseColor_Specified, specifiedColor=<rgba 0 0 0 0>}
      cornerRadii=<0 0 0 0>
     }
     tintColor {colorUseRule=UseColor_Specified, specifiedColor=<rgba 1 1 1 1>}
     uVRegion {bIsValid=false, max=<0 0>, min=<0 0>}
    }
    padding=<ltrb 0 -6 0 0>
   }
   slot {class=HorizontalBoxSlot}
  }
  Image_4 {
   widgetClassPath=&Image
   bIsVariable=true
   parent=Border_2
   widget=Image_4
   properties {brush {
     bIsDynamicallyLoaded=false
     drawAs=Image
     imageType=NoImage
     mirroring=NoMirror
     resourceName=None
     resourceObject=&M-Icon-ExternalLink-64
     tiling=NoTile
     imageSize=<76 76>
     margin=<ltrb 0 0 0 0>
     outlineSettings {
      bUseBrushTransparency=false
      roundingType=HalfHeightRadius
      width=0
      color {colorUseRule=UseColor_Specified, specifiedColor=<rgba 0 0 0 0>}
      cornerRadii=<0 0 0 0>
     }
     tintColor {colorUseRule=UseColor_Specified, specifiedColor=<rgba 1 1 1 1>}
     uVRegion {bIsValid=false, max=<0 0>, min=<0 0>}
    }}
   slot {class=BorderSlot, properties {horizontalAlignment=HAlign_Left, verticalAlignment=VAlign_Center, padding=<ltrb 0 -6 0 0>}}
  }
  Border_2 {
   widgetClassPath=&Border
   parent=HorizontalBox_2
   widget=Border_2
   properties {
    horizontalAlignment=HAlign_Left
    verticalAlignment=VAlign_Center
    background {
     bIsDynamicallyLoaded=false
     drawAs=NoDrawType
     imageType=NoImage
     mirroring=NoMirror
     resourceName=None
     resourceObject=null
     tiling=NoTile
     imageSize=<32 32>
     margin=<ltrb 0 0 0 0>
     outlineSettings {
      bUseBrushTransparency=false
      roundingType=HalfHeightRadius
      width=0
      color {colorUseRule=UseColor_Specified, specifiedColor=<rgba 0 0 0 0>}
      cornerRadii=<0 0 0 0>
     }
     tintColor {colorUseRule=UseColor_Specified, specifiedColor=<rgba 1 1 1 1>}
     uVRegion {bIsValid=false, max=<0 0>, min=<0 0>}
    }
    padding=<ltrb 0 -6 0 0>
   }
   slot {class=HorizontalBoxSlot}
  }
  Image_4 {
   widgetClassPath=&Image
   bIsVariable=true
   parent=Border_2
   widget=Image_4
   properties {brush {
     bIsDynamicallyLoaded=false
     drawAs=Image
     imageType=NoImage
     mirroring=NoMirror
     resourceName=None
     resourceObject=&M-Icon-ExternalLink-64
     tiling=NoTile
     imageSize=<76 76>
     margin=<ltrb 0 0 0 0>
     outlineSettings {
      bUseBrushTransparency=false
      roundingType=HalfHeightRadius
      width=0
      color {colorUseRule=UseColor_Specified, specifiedColor=<rgba 0 0 0 0>}
      cornerRadii=<0 0 0 0>
     }
     tintColor {colorUseRule=UseColor_Specified, specifiedColor=<rgba 1 1 1 1>}
     uVRegion {bIsValid=false, max=<0 0>, min=<0 0>}
    }}
   slot {class=BorderSlot, properties {horizontalAlignment=HAlign_Left, verticalAlignment=VAlign_Center, padding=<ltrb 0 -6 0 0>}}
  }
  TrademarksBorder {
   widgetClassPath=&Border
   bIsVariable=true
   parent=Overlay_0
   widget=TrademarksBorder
   properties {
    background {
     bIsDynamicallyLoaded=false
     drawAs=Image
     imageType=NoImage
     mirroring=NoMirror
     resourceName=None
     resourceObject=/FortUILibrary/Foundation/Buttons/Materials/M_SkewButtonShape_DF
     tiling=NoTile
     imageSize=<32 32>
     margin=<ltrb 0 0 0 0>
     outlineSettings {
      bUseBrushTransparency=false
      roundingType=HalfHeightRadius
      width=0
      color {colorUseRule=UseColor_Specified, specifiedColor=<rgba 0 0 0 0>}
      cornerRadii=<0 0 0 0>
     }
     tintColor {colorUseRule=UseColor_Specified, specifiedColor=<rgba 0.0052 0.0052 0.0056 0.87>}
     uVRegion {bIsValid=false, max=<0 0>, min=<0 0>}
    }
    padding=<ltrb 80 50 70 50>
   }
   slot {class=OverlaySlot, properties {horizontalAlignment=HAlign_Right, verticalAlignment=VAlign_Bottom}}
  }
  VerticalBox_5 {
   widgetClassPath=&VerticalBox
   parent=TrademarksBorder
   widget=VerticalBox_5
   slot {class=BorderSlot, properties {padding=<ltrb 80 50 70 50>}}
  }
  CommonTextBlock_0 {
   widgetClassPath=&CommonTextBlock
   parent=VerticalBox_5
   widget=CommonTextBlock_0
   properties {
    autoWrapText=true
    style=/FortUILibrary/Styles/Typography/Foundation/TextStyle-Base-S-70pc.TextStyle-Base-S-70pc_C
    text="“PlayStation” is a registered trademark of Sony Computer Entertainment Inc."
    colorAndOpacity {colorUseRule=UseColor_Specified, specifiedColor=<rgba 1 1 1 0.7>}
    font {
     bForceMonospaced=false
     bMaterialIsStencil=false
     fontMaterial=null
     fontObject=&NotoSans
     letterSpacing=0
     monospacedWidth=1
     size=18
     skewAmount=0
     typefaceFontName=NotoSans-Regular
     outlineSettings {
      bApplyOutlineToDropShadows=false
      bMiteredCorners=false
      bSeparateFillAlpha=false
      outlineMaterial=null
      outlineSize=0
      outlineColor=<rgba 0 0 0 1>
     }
    }
    shadowOffset=<0 0>
   }
   slot {class=VerticalBoxSlot}
  }
  CommonTextBlock_2 {
   widgetClassPath=&CommonTextBlock
   parent=VerticalBox_5
   widget=CommonTextBlock_2
   properties {
    autoWrapText=true
    style=&TextStyle-Base-S-70pc_C
    text="“Xbox Live” is a registered trademark of Microsoft Corporation."
    colorAndOpacity {colorUseRule=UseColor_Specified, specifiedColor=<rgba 1 1 1 0.7>}
    font {
     bForceMonospaced=false
     bMaterialIsStencil=false
     fontMaterial=null
     fontObject=&NotoSans
     letterSpacing=0
     monospacedWidth=1
     size=18
     skewAmount=0
     typefaceFontName=NotoSans-Regular
     outlineSettings {
      bApplyOutlineToDropShadows=false
      bMiteredCorners=false
      bSeparateFillAlpha=false
      outlineMaterial=null
      outlineSize=0
      outlineColor=<rgba 0 0 0 1>
     }
    }
    shadowOffset=<0 0>
   }
   slot {class=VerticalBoxSlot}
  }
 ]
}
