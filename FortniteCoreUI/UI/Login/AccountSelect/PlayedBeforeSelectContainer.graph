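# Login-flow container widget for the "played before" account selection screens, based on the native FortStartLoginContainer class.
# An activatable widget switcher holds an empty default spacer, the PlayedBeforeSelect screen and a PlayedBeforeSelectMP screen bound as Widget_StartLogin.
# The switcher and both screens bind to inherited parent variables.
# The event graph has only unconnected stub events, so the Blueprint supplies layout and adds no switching logic.
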WidgetBlueprint /FortniteCoreUI/UI/Login/AccountSelect/PlayedBeforeSelectContainer
graphs [EventGraph {nodes [
   Event PreConstruct()
   Event Construct()
   Event Tick()
  ]}]
variables []
widgets {
 info {
  inheritedWidgetCount=3
  parentClass=/Script/FortniteUI.FortStartLoginContainer
  rootWidgetClass=&CommonActivatableWidgetSwitcher
  widgetCount=4
 }
 widgets [
  Switcher_StartLogin {widgetClassPath=&CommonActivatableWidgetSwitcher, bInherited=true, bIsVariable=true, widget=Switcher_StartLogin}
  Spacer_Default {
   widgetClassPath=&Spacer
   parent=Switcher_StartLogin
   widget=Spacer_Default
   slot {class=WidgetSwitcherSlot, properties {horizontalAlignment=HAlign_Left, verticalAlignment=VAlign_Top}}
  }
  Widget_PlayedBeforeSelect {
   widgetClassPath=&/PlayedBeforeSelect
   bInherited=true
   bIsVariable=true
   parent=Switcher_StartLogin
   widget=Widget_PlayedBeforeSelect
   slot {class=WidgetSwitcherSlot}
  }
  Widget_StartLogin {
   widgetClassPath=&/PlayedBeforeSelectMP
   bInherited=true
   bIsVariable=true
   parent=Switcher_StartLogin
   widget=Widget_StartLogin
   slot {class=WidgetSwitcherSlot}
  }
 ]
}
