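# Login-time corrective action flow screen, a layout-only Widget Blueprint of FortCorrectiveActionFlow with no Blueprint logic.
# An inherited CommonActivatableWidgetSwitcher with zero transition time holds date of birth, guardian email, display name, EULA and unsupported-action screens for the native parent.
# A safe-zone column frames the switcher between the Fortnite logo and rich-text LinkText with placeholder text.
# Behind it sit a radial gradient background and a post-buffer capture widget.
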WidgetBlueprint /FortniteCoreUI/UI/Login/CorrectiveAction/CorrectiveActionFlow
graphs [EventGraph {nodes [
   Event PreConstruct()
   Event Construct()
   Event Tick()
  ]}]
variables []
widgets {
 info {inheritedWidgetCount=7, parentClass=/Script/FortniteUI.FortCorrectiveActionFlow, rootWidgetClass=&Overlay, widgetCount=14}
 widgets [
  Overlay_27 {widgetClassPath=&Overlay, widget=Overlay_27}
  ScaleBox {
   widgetClassPath=&ScaleBox
   parent=Overlay_27
   widget=ScaleBox
   properties {stretch=ScaleToFill}
   slot {class=OverlaySlot, properties {horizontalAlignment=HAlign_Fill, verticalAlignment=VAlign_Fill}}
  }
  BackgroundGradient {
   widgetClassPath=&Image
   parent=ScaleBox
   widget=BackgroundGradient
   properties {
    visibility=SelfHitTestInvisible
    brush {
     bIsDynamicallyLoaded=false
     drawAs=Image
     imageType=NoImage
     mirroring=NoMirror
     resourceName=None
     resourceObject=/Game/Athena/UI/Frontend/Art/M_UI_RadialGradientMaster_Inst
     tiling=NoTile
     imageSize=<32 32>
     margin=<ltrb 0 0 0 0>
     outlineSettings {
      bUseBrushTransparency=false
      roundingType=HalfHeightRadius
      width=0
      color {colorUseRule=UseColor_Specified, specifiedColor=<rgba 0 0 0 0>}
      cornerRadii=<0 0 0 0>
     }
     tintColor {colorUseRule=UseColor_Specified, specifiedColor=<rgba 1 1 1 1>}
     uVRegion {bIsValid=false, max=<0 0>, min=<0 0>}
    }
   }
   slot {class=ScaleBoxSlot, properties {horizontalAlignment=HAlign_Fill, verticalAlignment=VAlign_Fill}}
  }
  WBP_CaptureForPostBufferUpdate {
   widgetClassPath=&WBP_CaptureForPostBufferUpdate
   bIsVariable=true
   parent=Overlay_27
   slot {class=OverlaySlot, properties {horizontalAlignment=HAlign_Fill, verticalAlignment=VAlign_Fill}}
  }
  SafeZone {
   widgetClassPath=&SafeZone
   bIsVariable=true
   parent=Overlay_27
   widget=SafeZone
   slot {class=OverlaySlot, properties {horizontalAlignment=HAlign_Fill, verticalAlignment=VAlign_Fill}}
  }
  VerticalBox {
   widgetClassPath=&VerticalBox
   parent=SafeZone
   widget=VerticalBox
   slot {class=SafeZoneSlot, properties {padding=<ltrb 0 64 0 64>}}
  }
  FortniteLogo {
   widgetClassPath=&Image
   bIsVariable=true
   parent=VerticalBox
   widget=FortniteLogo
   properties {brush {
     bIsDynamicallyLoaded=false
     drawAs=Image
     imageType=NoImage
     mirroring=NoMirror
     resourceName=None
     resourceObject=/FortniteCoreUI/UI/Login/Art/MI_UI_Logo_Fortnite
     tiling=NoTile
     imageSize=<460 128>
     margin=<ltrb 0 0 0 0>
     outlineSettings {
      bUseBrushTransparency=false
      roundingType=HalfHeightRadius
      width=0
      color {colorUseRule=UseColor_Specified, specifiedColor=<rgba 0 0 0 0>}
      cornerRadii=<0 0 0 0>
     }
     tintColor {colorUseRule=UseColor_Specified, specifiedColor=<rgba 1 1 1 1>}
     uVRegion {bIsValid=false, max=<0 0>, min=<0 0>}
    }}
   slot {class=VerticalBoxSlot, properties {horizontalAlignment=HAlign_Center, verticalAlignment=VAlign_Top}}
  }
  CorrectiveActionScreenSwitcher {
   widgetClassPath=&CommonActivatableWidgetSwitcher
   bInherited=true
   bIsVariable=true
   parent=VerticalBox
   widget=CorrectiveActionScreenSwitcher
   properties {transitionDuration=0}
   slot {class=VerticalBoxSlot, properties {padding=<ltrb 0 64 0 0>, size {sizeRule=Fill, value=1}}}
  }
  DateOfBirthScreen {
   widgetClassPath=&/CorrectiveActionDateOfBirthScreen
   bInherited=true
   bIsVariable=true
   parent=CorrectiveActionScreenSwitcher
   widget=DateOfBirthScreen
   slot {class=WidgetSwitcherSlot, properties {horizontalAlignment=HAlign_Center}}
  }
  GuardianEmailScreen {
   widgetClassPath=&/CorrectiveActionGuardianEmailScreen
   bInherited=true
   bIsVariable=true
   parent=CorrectiveActionScreenSwitcher
   widget=GuardianEmailScreen
   slot {class=WidgetSwitcherSlot, properties {horizontalAlignment=HAlign_Center}}
  }
  DisplayNameScreen {
   widgetClassPath=&/CorrectiveActionDisplayNameScreen
   bInherited=true
   bIsVariable=true
   parent=CorrectiveActionScreenSwitcher
   widget=DisplayNameScreen
   slot {class=WidgetSwitcherSlot, properties {horizontalAlignment=HAlign_Center}}
  }
  EulaScreen {
   widgetClassPath=&/CorrectiveActionEULAScreen
   bInherited=true
   bIsVariable=true
   parent=CorrectiveActionScreenSwitcher
   widget=EulaScreen
   slot {class=WidgetSwitcherSlot}
  }
  UnsupportedActionScreen {
   widgetClassPath=&/CorrectiveActionUnsupportedActionScreen
   bInherited=true
   bIsVariable=true
   parent=CorrectiveActionScreenSwitcher
   widget=UnsupportedActionScreen
   slot {class=WidgetSwitcherSlot, properties {horizontalAlignment=HAlign_Center}}
  }
  LinkText {
   widgetClassPath=/Script/FortniteUI.FortRichTextBlockLegacy
   bInherited=true
   bIsVariable=true
   parent=VerticalBox
   widget=LinkText
   properties {
    justification=Center
    styleSet=/Game/UI/Foundation/Legal/LegalScreenRichTextStyles
    text="[PH] Temp Text"
    hyperlinkButtonStyle {
     clickedSlateSound {resourceObject=null}
     disabled {
      bIsDynamicallyLoaded=false
      drawAs=NoDrawType
      imageType=NoImage
      mirroring=NoMirror
      resourceName=None
      resourceObject=null
      tiling=NoTile
      imageSize=<0 0>
      margin=<ltrb 0 0 0 0>
      outlineSettings {
       bUseBrushTransparency=false
       roundingType=HalfHeightRadius
       width=0
       color {colorUseRule=UseColor_Specified, specifiedColor=<rgba 0 0 0 0>}
       cornerRadii=<0 0 0 0>
      }
      tintColor {colorUseRule=UseColor_Specified, specifiedColor=<rgba 1 1 1 1>}
      uVRegion {bIsValid=false, max=<0 0>, min=<0 0>}
     }
     disabledForeground {colorUseRule=UseColor_Foreground, specifiedColor=<rgba 1 0 1 1>}
     hovered {
      bIsDynamicallyLoaded=false
      drawAs=Image
      imageType=NoImage
      mirroring=NoMirror
      resourceName=None
      resourceObject=null
      tiling=NoTile
      imageSize=<32 32>
      margin=<ltrb 0 0 0 0>
      outlineSettings {
       bUseBrushTransparency=false
       roundingType=HalfHeightRadius
       width=0
       color {colorUseRule=UseColor_Specified, specifiedColor=<rgba 0 0 0 0>}
       cornerRadii=<0 0 0 0>
      }
      tintColor {colorUseRule=UseColor_Specified, specifiedColor=<rgba 1 1 1 0>}
      uVRegion {bIsValid=false, max=<0 0>, min=<0 0>}
     }
     hoveredForeground {colorUseRule=UseColor_Foreground, specifiedColor=<rgba 1 0 1 1>}
     hoveredSlateSound {resourceObject=null}
     normal {
      bIsDynamicallyLoaded=false
      drawAs=Image
      imageType=NoImage
      mirroring=NoMirror
      resourceName=None
      resourceObject=null
      tiling=NoTile
      imageSize=<32 32>
      margin=<ltrb 0 0 0 0>
      outlineSettings {
       bUseBrushTransparency=false
       roundingType=HalfHeightRadius
       width=0
       color {colorUseRule=UseColor_Specified, specifiedColor=<rgba 0 0 0 0>}
       cornerRadii=<0 0 0 0>
      }
      tintColor {colorUseRule=UseColor_Specified, specifiedColor=<rgba 1 1 1 0>}
      uVRegion {bIsValid=false, max=<0 0>, min=<0 0>}
     }
     normalForeground {colorUseRule=UseColor_Foreground, specifiedColor=<rgba 1 0 1 1>}
     normalPadding=<ltrb 0 0 0 0>
     pressed {
      bIsDynamicallyLoaded=false
      drawAs=Image
      imageType=NoImage
      mirroring=NoMirror
      resourceName=None
      resourceObject=null
      tiling=NoTile
      imageSize=<32 32>
      margin=<ltrb 0 0 0 0>
      outlineSettings {
       bUseBrushTransparency=false
       roundingType=HalfHeightRadius
       width=0
       color {colorUseRule=UseColor_Specified, specifiedColor=<rgba 0 0 0 0>}
       cornerRadii=<0 0 0 0>
      }
      tintColor {colorUseRule=UseColor_Specified, specifiedColor=<rgba 1 1 1 1>}
      uVRegion {bIsValid=false, max=<0 0>, min=<0 0>}
     }
     pressedForeground {colorUseRule=UseColor_Foreground, specifiedColor=<rgba 1 0 1 1>}
     pressedPadding=<ltrb 0 0 0 0>
     pressedSlateSound {resourceObject=null}
    }
   }
   slot {class=VerticalBoxSlot, properties {verticalAlignment=VAlign_Top, padding=<ltrb 0 24 0 0>}}
  }
 ]
}
