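# EulaWidgetContainer is the login EULA container widget, built on the native FortEulaWidgetsContainer class.
# An activatable widget switcher at the root holds a default spacer plus old and new EULA screens, using EulaWidget and EulaWidgetMP respectively.
# The Blueprint graph has only unconnected stock events, and the switcher and both EULA widgets are inherited bindings from the parent class.
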WidgetBlueprint /FortniteCoreUI/UI/Login/Eula/EulaWidgetContainer
graphs [EventGraph {nodes [
   Event PreConstruct()
   Event Construct()
   Event Tick()
  ]}]
variables []
widgets {
 info {
  inheritedWidgetCount=3
  parentClass=/Script/FortniteUI.FortEulaWidgetsContainer
  rootWidgetClass=&CommonActivatableWidgetSwitcher
  widgetCount=4
 }
 widgets [
  Switcher_EulaWidgets {widgetClassPath=&CommonActivatableWidgetSwitcher, bInherited=true, bIsVariable=true, widget=Switcher_EulaWidgets}
  Spacer_Default {
   widgetClassPath=&Spacer
   parent=Switcher_EulaWidgets
   widget=Spacer_Default
   slot {class=WidgetSwitcherSlot, properties {horizontalAlignment=HAlign_Left, verticalAlignment=VAlign_Top}}
  }
  Widget_OldEula {
   widgetClassPath=&/EulaWidget
   bInherited=true
   bIsVariable=true
   parent=Switcher_EulaWidgets
   widget=Widget_OldEula
   slot {class=WidgetSwitcherSlot}
  }
  Widget_NewEula {
   widgetClassPath=&/EulaWidgetMP
   bInherited=true
   bIsVariable=true
   parent=Switcher_EulaWidgets
   widget=Widget_NewEula
   slot {class=WidgetSwitcherSlot}
  }
 ]
}
